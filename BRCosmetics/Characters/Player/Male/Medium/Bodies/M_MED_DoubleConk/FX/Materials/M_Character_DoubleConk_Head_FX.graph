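# Masked, lit material for the DoubleConk character's head FX on Niagara mesh particles, drawing a glowing grid.
# Emissive blends grid line and cell colors via UV3, scaled by Fresnel and darkened along the trim.
# A mask texture drives dithered temporal-AA opacity, while panning noise, masked and edge-faded, displaces vertices along axis-weighted pre-skinned normals, plus a camera-ward offset.
# Screen-derivative faceted normals blend with flat normals by UV3.
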Material /BRCosmetics/Characters/Player/Male/Medium/Bodies/M_MED_DoubleConk/FX/Materials/M_Character_DoubleConk_Head_FX
graph {
 properties {bAutomaticallySetUsageInEditor=false, bUsedWithNiagaraMeshParticles=true, blendMode=BLEND_Masked, usedShadingModels=Default Lit}
 nodes [
  n1 = NamedRerouteUsage()
  n2 = Power(
   Base=DotProduct(A=Transform(None=@n1), B=CameraVectorWS())
   Exp=ScalarParameter() {defaultValue=1.25, parameterName=Emissive Fresnel Exponent}
  )
  n3 = VectorParameter() {parameterName=Grid Lines Emissive Color, defaultValue=<rgba 0 0.4452 0.5089 1>}
  n4 = VectorParameter() {parameterName=Grid Cells Emissive Color, defaultValue=<rgba 0.0452 0.2232 0.4179 1>}
  n5 = VectorParameter() {parameterName=Grid Cells Emissive Intensity Multiplier, defaultValue=<rgba 0.4 4 0 1.75>}
  n6 = NamedRerouteUsage()
  n7 = Power(Base=ComponentMask(None=TextureCoordinate() {coordinateIndex=3}) {r=true})
  n8 = LinearInterpolate(A=@n5.R, B=@n5.R, Alpha=Power(Base=Multiply(A=@n6, B=@n7), Exp=@n5.A))
  n9 = ComponentMask(None=TextureCoordinate() {coordinateIndex=3}) {r=true}
  n10 = Multiply(
   A=LinearInterpolate(A=@n3.RGB, B=Multiply(A=@n4.RGB, B=@n8), Alpha=Power(Base=@n9, Exp=@n3.A))
   B=ScalarParameter() {defaultValue=2.5, parameterName=Emissive Brightness}
  )
  n11 = TextureSample() {
   samplerSource=SSM_Clamp_WorldGroupSettings
   samplerType=SAMPLERTYPE_LinearGrayscale
   texture=/BRCosmetics/Characters/Player/Male/Medium/Bodies/M_MED_DoubleConk/FX/Textures/T_Character_DoubleConk_Head_FX_Mask
  }
  n12 = Saturate(None=Divide(A=Subtract(A=@n11.R) {constB=0.5}, B=ScalarParameter() {defaultValue=0.05, parameterName=Trim Darkening Size}) {constB=0.5})
  n13 = OneMinus(None=Multiply(A=OneMinus(None=@n12), B=ScalarParameter() {defaultValue=0.2, parameterName=Trim Darkening Intensity}))
  MaterialFunctionCall(Emissive=Multiply(A=Multiply(A=@n2, B=@n10), B=@n13)) {materialFunction=&MF_FX_EmissiveAdjustments}
  Multiply(A=VectorParameter() {parameterName=Base Color, defaultValue=<rgba 0.0087 0.1403 0.3 0.6>}.RGB, B=@n13)
  n14 = Subtract(A=CameraPositionWS(), B=WorldPosition().XYZ)
  NamedRerouteDeclaration(None=Multiply(A=Normalize(VectorInput=@n14), B=ScalarParameter() {defaultValue=0.333, parameterName=Camera Offset})) {name=Camera Offset WPO, nodeColor=<rgba 0.1484 0 1 1>}
  n15 = TextureCoordinate() {coordinateIndex=1}
  n16 = ComponentMask(None=@n15) {r=true}
  n17 = OneMinus(None=Abs(None=ConstantBiasScale(None=@n16) {bias=-0.5, scale=2}))
  n18 = ScalarParameter() {defaultValue=1, parameterName=WPO Intensity Curve Exponent U}
  n19 = ComponentMask(None=@n15) {g=true}
  n20 = OneMinus(None=Abs(None=ConstantBiasScale(None=@n19) {bias=-0.5, scale=2}))
  n21 = ScalarParameter() {defaultValue=1, parameterName=WPO Intensity Curve Exponent V}
  NamedRerouteDeclaration(None=Multiply(A=Power(Base=@n17, Exp=@n18), B=Power(Base=@n20, Exp=@n21))) {name=WPO Intensity, nodeColor=<rgba 0.7812 1 0 1>}
  n22 = TextureCoordinate() {coordinateIndex=2}
  n23 = TextureSample(UVs=TextureCoordinate() {coordinateIndex=1}) {
   samplerType=SAMPLERTYPE_LinearGrayscale
   texture=/BRCosmetics/Characters/Player/Male/Medium/Bodies/M_MED_DoubleConk/FX/Textures/T_Character_DoubleConk_Head_FX_VertexMask
  }
  NamedRerouteDeclaration(None=Multiply(A=ComponentMask(None=@n22) {r=true}, B=@n23.R)) {name=WPO Mask, nodeColor=<rgba 0 0.0625 1 1>}
  n24 = WorldPosition()
  n25 = CrossProduct(A=DDX(Value=@n24.XYZ), B=DDY(Value=@n24.XYZ))
  NamedRerouteDeclaration(None=Transform(None=Normalize(VectorInput=@n25)) {transformSourceType=TRANSFORMSOURCE_World, transformType=TRANSFORM_Tangent}) {name=Faceted Normals, nodeColor=<rgba 0 1 0.9766 1>}
  n26 = TextureSample() {
   samplerSource=SSM_Clamp_WorldGroupSettings
   samplerType=SAMPLERTYPE_LinearGrayscale
   texture=&T_Character_DoubleConk_Head_FX_Mask
  }
  n27 = TextureCoordinate()
  n28 = Subtract(
   A=@n26.R
   B=LinearInterpolate(
    A=ScalarParameter() {defaultValue=0.01, parameterName=Opacity Mask Shift Back}
    B=ScalarParameter() {defaultValue=-0.015, parameterName=Opacity Mask Shift Front}
    Alpha=ComponentMask(None=@n27) {g=true}
   )
  )
  n29 = ConstantBiasScale(None=Divide(A=ConstantBiasScale(None=@n28) {bias=-0.5, scale=2}, B=ScalarParameter() {defaultValue=0.02, parameterName=Opacity Softness}))
  NamedRerouteDeclaration(None=MaterialFunctionCall(Alpha Threshold=Saturate(None=@n29)) {materialFunction=&DitherTemporalAA_Fortnite}.Result) {name=Shape Mask, nodeColor=<rgba 0 0.6719 1 1>}
  n30 = VectorParameter() {parameterName=Local Axis Influence, defaultValue=<rgba 0.2 1 1 0>}
  n31 = PreSkinnedNormal()
  n32 = Multiply(
   A=ScalarParameter() {defaultValue=3.6666, parameterName=WPO Intensity Max}
   B=Transform(None=Multiply(A=@n30.RGB, B=@n31)) {transformSourceType=TRANSFORMSOURCE_Local}
  )
  n33 = TextureCoordinate() {coordinateIndex=1}
  n34 = Panner(Coordinate=@n33) {speedX=0.2, speedY=0.8}
  n35 = Panner(Coordinate=@n33) {speedX=0.3, speedY=-0.7}
  n36 = Divide(A=Add(
    A=TextureSample(UVs=@n34) {samplerType=SAMPLERTYPE_Grayscale, texture=/Game/TimeOfDay/TODM_Sequence/Materials/Sky/Textures/LowResBlurredNoise_Radish_LF_Root}.R
    B=TextureSample(UVs=@n35) {samplerType=SAMPLERTYPE_Grayscale, texture=&LowResBlurredNoise_Radish_LF_Root}.R
   ))
  n37 = Multiply(
   A=Multiply(A=NamedRerouteDeclaration(None=@n36) {name=WPO Noise, nodeColor=<rgba 0.5937 1 0 1>}, B=NamedRerouteUsage())
   B=NamedRerouteUsage()
  )
  NamedRerouteDeclaration(None=Add(A=Multiply(A=@n32, B=@n37), B=NamedRerouteUsage())) {name=WPO, nodeColor=<rgba 0 0.8359 1 1>}
  n38 = TextureCoordinate() {coordinateIndex=3}
  n39 = Divide(A=OneMinus(None=ComponentMask(None=@n38) {r=true}), B=ScalarParameter() {defaultValue=0.05, parameterName=Normal Blend Softness})
  NamedRerouteDeclaration(None=LinearInterpolate(A=Constant3Vector() {constant=<rgba 0 0 1 0>}, B=NamedRerouteUsage(), Alpha=Saturate(None=@n39))) {name=Blended Normals, nodeColor=<rgba 0.5703 1 0 1>}
  NamedRerouteUsage()
  ScalarParameter() {defaultValue=0.73, parameterName=Metallic}
  n40 = NamedRerouteUsage()
  ShadowReplace(Default=@n40, Shadow=Multiply(A=@n40, B=NamedRerouteUsage()))
  VertexNormalWS()
 ]
}
