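# Idle cosmetic Niagara effect for the DoubleConk character's mask, with glowing derez cubes, driven by MaskOff and Style.
# Derez_Mesh pins a looping mask mesh to the skeletal mesh; Derez_Cubes_Mask streams swirling, dive-velocity-aligned cubes and sprites from it.
# When MaskOff and Style are both 1, Derez_Cube_Head instead sheds cubes around the head, faster with movement.
# BigHeadMode scales the effect; visibility tags, vehicle and camera-proximity checks hide parts.
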NiagaraSystem /BRCosmetics/Characters/Player/Male/Medium/Bodies/M_MED_DoubleConk/FX/NS_DoubleConk_Mask_Idle
dependencies {
 usedDataInterfaces [
  &NiagaraDataInterfaceCamera
  &NiagaraDataInterfaceCurve
  &NiagaraDataInterfaceVectorCurve
  &NiagaraDataInterfaceArrayFloat
  &NiagaraDataInterfaceVector2DCurve
  &NiagaraDataInterfaceSkeletalMesh
 ]
 usedDynamicInputs [&NormalizeDistanceRange, &NormalizeVector]
 usedModules [
  &SystemState
  &self:SystemSpawnScript.SystemScriptSource.SystemScriptGraph.NiagaraNodeAssignment_4.SetVariables_#1
  &InterpolateOverTime
  &self:SystemSpawnScript.SystemScriptSource.SystemScriptGraph.NiagaraNodeAssignment_3.SetVariables_#2
  &self:SystemSpawnScript.SystemScriptSource.SystemScriptGraph.NiagaraNodeAssignment_0.SetVariables_#3
 ]
}
emitters {
 Derez_Cube_Head {
  emitterUpdateScript [
   &EmitterState()
   /SetVariables(Emitter.MovementSpawnAlpha:NiagaraFloat=&ClampFloat(Float=&Divide_Float(A=&VectorLength(Vector=$Engine.Owner.Velocity), B=500))) {isSetParameters=true}
   &SpawnRate(
    Spawn Probability=&MakeCustomFloatFromBool(
     Bool=&Boolean_AndOperation(
      Bool1=&SetBoolByFloatComparison(A=$User.MaskOff, B=1, Comparison Type=::NewEnumerator4)
      Bool2=&SetBoolByFloatComparison(A=$User.Style, B=1, Comparison Type=::NewEnumerator4)
     )
     True Float=1
     False Float=0
    )
    SpawnRate=&Lerp_Float(A=5, B=10, Alpha=$Emitter.MovementSpawnAlpha)
    Use Spawn Probability=true
   )
   /SetVariables(Emitter.ADS:NiagaraBool=&SetBoolByFloatComparison(
     A=&DistanceBetweenPositions(
      A=&ReturnCameraPosition(
       Coordinate Space:ENiagaraCoordinateSpace=::Simulation
       Camera Query:NiagaraDataInterfaceCamera=DataInterface() {propertyValues {bRequireCurrentFrameData=false}}
      )
      B=$Engine.Emitter.SimulationPosition
     )
     B=&FN_FrontEndFloatModifier(Gameplay=200, Front End=0)
     Comparison Type=::NewEnumerator6
    )) {isSetParameters=true, module=SetVariables_#4}
  ]
  particleSpawnScript [
   &InitializeParticle(
    Lifetime=&Multiply_Float(
     A=&Lerp_Float(
      A=&RandomRangeFloat(Minimum=0.9, Maximum=1.1)
      B=1.5
      Alpha=&Divide_Float(A=&VectorLength(Vector=$Engine.Owner.Velocity), B=5000)
     )
     B=0.6
    )
    Mesh Scale Mode=::NewEnumerator1
    Mesh Uniform Scale Max=0.5
    Mesh Uniform Scale Min=0.375
    Sprite Size Mode=::NewEnumerator1
    Uniform Sprite Size Max=8
    Uniform Sprite Size Min=6
    UsePositionOffset=false
    Write Material Random=true
   )
   &InitialMeshOrientation(Use Rotation Vector=false)
   &SkeletalMeshLocation(
    Bone Sampling Mode=::NewEnumerator4
    MaskSpawnGroup=false
    Orientation Sampling=::NewEnumerator1
    Sampled Position Offset=&Lerp_Vector(A=<0 3 0>, B=<0 6 0>, Alpha=$NPC.FortniteNPC.BigHeadMode)
    Skeletal Mesh=$System.SkeletalMesh
   )
   &ShapeLocation(
    Additional Quaternion Rotation=true
    Additional Yaw / Pitch / Roll=<0.16 0 0>
    Apply To Particle Position=::NewEnumerator1
    Handle Radius=1
    Large Radius=&Lerp_Float(A=14, B=30, Alpha=$NPC.FortniteNPC.BigHeadMode)
    MaskSpawnGroup=false
    Non Uniform Scale=<0.9 1 7>
    Override Seed=false
    Rotation Mode=::NewEnumerator3
    Rotation Quaternion=$Particles.SkeletalMeshLocation.SampledRotation
    Shape Primitive=::NewEnumerator3
    U Distribution=0.55
    UseOwnerScale=true
   ) {module=ShapeLocation001}
   &AddVelocity(
    Velocity=&Lerp_Vector(A=$Particles.ShapeLocation001.ShapeNormal, B=<0 0 1>, Alpha=0)
    Velocity Speed Scale=&Multiply_Float(A=50, B=&MakeCustomFloatFromBool(Bool=$User.IsSkydiving, True Float=5, False Float=1))
   ) {module=AddVelocity003}
   /SetVariables(Particles.Position=$Particles.ShapeLocation001.ShapePosition) {isSetParameters=true}
  ]
  particleUpdateScript [
   &ParticleState()
   &ScaleMeshSize(Scale Factor=&VectorFromFloat(Value=&MakeCustomFloatFromBool(
      Bool=&Boolean_OrOperation(Bool1=$User.bisInVehicle, Bool2=$Emitter.ADS)
      True Float=0
      False Float=&FloatFromCurve(
       FloatCurve=DataInterface() {propertyValues {Curve {
          defaultValue=0
          keys [
           {arriveTangent=9.0725, interpMode=RCIM_Cubic, leaveTangent=9.0725, tangentMode=RCTM_User, time=0, value=0}
           {arriveTangent=0.1035, interpMode=RCIM_Cubic, leaveTangent=0.1035, tangentMode=RCTM_User, time=0.1408, value=0.9847}
           {arriveTangent=0.0495, interpMode=RCIM_Cubic, leaveTangent=0.0495, tangentMode=RCTM_User, time=0.7, value=1}
           {arriveTangent=-3.1446, interpMode=RCIM_Cubic, leaveTangent=-3.1446, tangentMode=RCTM_User, time=1, value=0}
          ]
         }}}
       Scale Curve=&Lerp_Float(A=1, B=2, Alpha=$NPC.FortniteNPC.BigHeadMode)
      )
     )))
   &ScaleSpriteSize(
    Scale Sprite Size Mode=::NewEnumerator0
    Uniform Scale Factor=&FloatFromCurve(
     FloatCurve=DataInterface() {propertyValues {Curve {keys [
         {arriveTangent=9.1187, interpMode=RCIM_Cubic, time=0, value=0}
         {arriveTangent=1.1429, interpMode=RCIM_Cubic, leaveTangent=1.1429, time=0.2487, value=1.1246}
         {arriveTangent=-1.497, interpMode=RCIM_Cubic, leaveTangent=-1.497, time=0.7, value=0.8}
         {arriveTangent=-3.0554, interpMode=RCIM_Cubic, leaveTangent=-3.0554, tangentMode=RCTM_User, time=1, value=0}
        ]}}}
     Scale Curve=&Lerp_Float(A=1, B=2, Alpha=$NPC.FortniteNPC.BigHeadMode)
    )
   )
   &AccelerationForce(Acceleration=&Multiply_VectorByFloat(Vector=&RandomRangeVector(Minimum=<-3 -3 1>, Maximum=<3 3 1>), Float=1))
   &Drag(Drag=10, Use Linear Drag=true, Use Rotational Drag=true)
   &GravityForce(Gravity=&MakeVector(
     X=0
     Y=0
     Z=&FloatFromCurve(FloatCurve=DataInterface() {propertyValues {Curve {keys [<at 0.3 0>,<at 1 1>]}}}, Scale Curve=200)
    ))
   &DynamicMaterialParameters(
    Index 0 Param 1=&FloatFromCurve(
     FloatCurve=DataInterface() {propertyValues {Curve {
        defaultValue=1
        keys [
         {arriveTangent=0.0618, interpMode=RCIM_Cubic, leaveTangent=0.0618, tangentMode=RCTM_User, time=0, value=0}
         <at 0.3 0.1125 RCIM_Cubic>
         {interpMode=RCIM_Cubic, tangentMode=RCTM_User, time=1, value=0}
        ]
       }}}
     Scale Curve=0.5
    )
    Index 0 Param 3=&RandomRangeFloat(Minimum=0.75, Maximum=1.75)
    Module.Param0WriteEnabled=true
    Module.Param1WriteEnabled=true
    Module.Param2WriteEnabled=true
    Module.Param3WriteEnabled=false
   )
   &InheritSourceMovement(
    Applied Force Scale=&VectorFromFloat(Value=&FloatFromCurve(FloatCurve=DataInterface() {propertyValues {Curve {keys [<at 0 0>,<at 1 -0.5>]}}}))
    Applied Position Scale=&VectorFromFloat(Value=&MakeCustomFloatFromBool(
      Bool=$User.IsSkydiving
      True Float=&FloatFromCurve(FloatCurve=DataInterface() {propertyValues {Curve {keys [<at 0 0.95 RCIM_Cubic>,<at 1 0.8 RCIM_Cubic>]}}})
      False Float=&FloatFromCurve(FloatCurve=DataInterface() {propertyValues {Curve {keys [<at 0 0.95 RCIM_Cubic>, {arriveTangent=-0.2785, interpMode=RCIM_Cubic, leaveTangent=-0.2785, tangentMode=RCTM_User, time=1, value=0}]}}})
     ))
    ApplyForce=false
    ApplyPosition=true
    ApplyPositionFalloff=false
   ) {module=InheritSourceMovement001}
   &PointForce(Force Origin=$System.BodyKillPos, Force Strength=50, Module.Use Falloff=false, Module.UseLinearForce=false)
   &SolveForcesAndVelocity(Clamp Velocity=false, Limit Acceleration=false)
   &ScaleColor(
    Scale Alpha=&MakeCustomFloatFromBool(Bool=&Boolean_OrOperation(Bool1=$User.bisInVehicle, Bool2=$Emitter.ADS), True Float=0, False Float=1)
    ScaleA=true
    ScaleRGB=true
   ) {module=ScaleColor001}
   &ScaleColor(
    Scale Alpha=&FloatFromCurve(FloatCurve=DataInterface() {propertyValues {Curve {
        defaultValue=0
        keys [
         {interpMode=RCIM_Cubic, tangentMode=RCTM_User, time=0, value=0}
         {arriveTangent=0.3339, interpMode=RCIM_Cubic, leaveTangent=0.3339, tangentMode=RCTM_User, time=0.1, value=0.9}
         {interpMode=RCIM_Cubic, tangentMode=RCTM_User, time=0.7, value=0.9}
         {tangentMode=RCTM_User, time=1, value=0}
        ]
       }}})
    Scale RGB=&VectorFromFloat(Value=&FNTimeOfDayScalar(Night Scale=0.6, Morning Scale=1, Evening Scale=1))
    ScaleA=true
    ScaleRGB=true
   ) {module=ScaleColor002}
   &CameraOffset(Camera Offset Amount=&Multiply_Float(
     A=&FloatFromCurve(FloatCurve=DataInterface() {propertyValues {Curve {keys [{interpMode=RCIM_Cubic, tangentMode=RCTM_User, time=0.5, value=1}, {interpMode=RCIM_Cubic, tangentMode=RCTM_User, time=1, value=0}]}}})
     B=0
    ))
   /SetVariables(Particles.VisibilityTag=&MakeCustomIntFromBool(Bool=$User.bisHeadHidden, True INT=1, False INT=0)) {isSetParameters=true, module=SetVariables_#5}
   /SetVariables(Particles.Color_Glow:LinearColor=&MultiplyLinearColor(
     Linear Color=<rgba 0.0681 0.7611 1 0.1>
     Scale Factor=&MakeVector4_FromLinearColor(LinearColor=$Particles.Color)
    )) {isSetParameters=true, module=SetVariables_#6}
  ]
  rendererClasses [&NiagaraMeshRendererProperties, &NiagaraSpriteRendererProperties]
  renderers [
   {rendererClass=&NiagaraMeshRendererProperties, rendererIndex=0}
   {rendererClass=&NiagaraSpriteRendererProperties, rendererIndex=1}
  ]
 }
 Derez_Cubes_Mask {
  emitterUpdateScript [
   &EmitterState()
   &SpawnRate(
    Spawn Probability=&MakeCustomFloatFromBool(
     Bool=&SetBoolByFloatComparison(A=$User.MaskOff, B=0, Comparison Type=::NewEnumerator4)
     True Float=1
     False Float=0
    )
    SpawnRate=20
    Use Spawn Probability=true
   ) {module=SpawnRate001}
   &SpawnRate(
    Spawn Probability=&MakeCustomFloatFromBool(
     Bool=&SetBoolByFloatComparison(A=$User.MaskOff, B=0, Comparison Type=::NewEnumerator4)
     True Float=1
     False Float=0
    )
    SpawnGroup=1
    SpawnRate=10
    Use Spawn Probability=true
   ) {module=SpawnRate002}
   /SetVariables(Emitter.ADS:NiagaraBool=&SetBoolByFloatComparison(
     A=&DistanceBetweenPositions(
      A=&ReturnCameraPosition(
       Coordinate Space:ENiagaraCoordinateSpace=::Simulation
       Camera Query:NiagaraDataInterfaceCamera=DataInterface() {propertyValues {bRequireCurrentFrameData=false}}
      )
      B=$Engine.Emitter.SimulationPosition
     )
     B=&FN_FrontEndFloatModifier(Gameplay=200, Front End=0)
     Comparison Type=::NewEnumerator6
    )) {enabled=false, isSetParameters=true, module=SetVariables_#4}
  ]
  particleSpawnScript [
   /SetVariables(Particles.SpawnGroup=$Emitter.SpawnGroup) {isSetParameters=true}
   &InitialMeshOrientation(Mesh Orientation Mode=::NewEnumerator2, Rotation Coordinate Space=::NewEnumerator2, Use Rotation Vector=false) {module=InitialMeshOrientation001}
   &InitializeParticle(
    Lifetime=&MergeHelper_FloatSwitch(
     0=&RandomRangeFloat(Minimum=1.25, Maximum=1.5)
     1=&RandomRangeFloat(Minimum=1.75, Maximum=2)
     2=?
     3=?
     Index=$Particles.SpawnGroup
     Num=::NewEnumerator0
    )
    Mesh Scale=&MakeVector(
     X=&RandomRangeFloat(Minimum=1.25, Maximum=1.5)
     Y=&RandomRangeFloat(Minimum=0.5, Maximum=1)
     Z=&RandomRangeFloat(Minimum=0.5, Maximum=1)
    )
    Mesh Scale Mode=::NewEnumerator4
    Sprite Size Mode=::NewEnumerator1
    Uniform Sprite Size Max=8
    Uniform Sprite Size Min=6
    UsePositionOffset=false
    Write Material Random=true
   )
  ]
  particleUpdateScript [
   &ParticleState()
   &SkeletalMeshLocation(
    Bone Sampling Mode=::NewEnumerator4
    MaskSpawnGroup=false
    Orientation Sampling=::NewEnumerator1
    Sampled Position Offset=&Lerp_Vector(A=<10 1.5 0>, B=<20 3 0>, Alpha=$NPC.FortniteNPC.BigHeadMode)
    Skeletal Mesh=$System.SkeletalMesh
   )
   &ShapeLocation(
    Additional Quaternion Rotation=true
    Additional Yaw / Pitch / Roll=<0.22 1 0>
    Apply To Particle Position=::NewEnumerator1
    Handle Radius=1
    Large Radius=8
    MaskSpawnGroup=true
    Non Uniform Scale=<1.2 1 1.5>
    Override Seed=false
    Rotation Mode=::NewEnumerator3
    Rotation Quaternion=$Particles.SkeletalMeshLocation.SampledRotation
    Shape Origin=$StackContext.Position
    Shape Primitive=::NewEnumerator3
    Spawn Group Mask=2
    U Distribution=0.6
    UseOwnerScale=true
   ) {module=ShapeLocation001}
   &SkeletalMeshLocation(
    Bone Sampling Mode=::NewEnumerator4
    MaskSpawnGroup=false
    Orientation Sampling=::NewEnumerator1
    Sampled Position Offset=&Multiply_VectorByFloat(
     Vector=&AddVectorToPosition(
      Position=<10 1.5 0>
      Vector=&MergeHelper_VectorSwitch(
       0=&MakeVector(
        X=&ClampFloat(
         Float=&Multiply_Float(
          A=&MakeFloatFromVector(Vector=Linked() {linkedVariable {name=Particles.ShapeLocation001.ShapePosition, type=NiagaraPosition}}, Channel=::NewEnumerator0)
          B=&RandomRangeFloat(Minimum=-1, Maximum=1)
         )
         Max=3
        )
        Y=&ClampFloat(
         Float=&Multiply_Float(
          A=&MakeFloatFromVector(Vector=Linked() {linkedVariable {name=Particles.ShapeLocation001.ShapePosition, type=NiagaraPosition}}, Channel=::NewEnumerator1)
          B=&RandomRangeFloat(Minimum=-1, Maximum=1)
         )
         Min=-3
         Max=4
        )
        Z=0
       )
       1=<0 0 0>
       2=?
       3=?
       Index=$Particles.SpawnGroup
       Num=::NewEnumerator0
      )
     )
     Float=&Lerp_Float(A=1, B=2, Alpha=$NPC.FortniteNPC.BigHeadMode)
    )
    Skeletal Mesh=$System.SkeletalMesh
   ) {module=SkeletalMeshLocation001}
   /SetVariables(Particles.Position=&Add_Vector(
     A=&ConvertPositionToVector(Input Position=$Particles.Position, Convert To Absolute Large World Vector=false)
     B=&Lerp_Vector(
      A=&Multiply_VectorByFloat(
       Vector=&Add_Vector(
        A=&Lerp_Vector(A=<0 0 1>, B=$System.DiveVelocityScale, Alpha=&Power(a=$Particles.NormalizedAge, b=3))
        B=&RotateVectorByQuaternion(Vector To Rotate=<0 -0.1 0>, Quaternion=$Particles.SkeletalMeshLocation001.SampledRotation)
       )
       Float=&FloatFromCurve(
        FloatCurve=DataInterface() {propertyValues {Curve {keys [
            {arriveTangent=0.5799, interpMode=RCIM_Cubic, leaveTangent=0.5799, tangentMode=RCTM_User, time=0, value=-0.1}
            {arriveTangent=1.4467, interpMode=RCIM_Cubic, leaveTangent=1.4467, tangentMode=RCTM_User, time=1, value=1}
           ]}}}
        Scale Curve=&MergeHelper_FloatSwitch(
         0=&RandomRangeFloat(Minimum=7.5, Maximum=10)
         1=15
         2=?
         3=?
         Index=$Particles.SpawnGroup
         Num=::NewEnumerator0
        )
       )
      )
      B=&RotateVectorByQuaternion(
       Vector To Rotate=&Multiply_Vector(
        A=$Particles.ShapeLocation001.ShapeNormal
        B=&VectorFromCurve(
         VectorCurve=DataInterface() {propertyValues {
           XCurve {
            defaultValue=1
            keys [
             {arriveTangent=0.0249, arriveTangentWeight=0.3333, interpMode=RCIM_Cubic, leaveTangent=0.0249, tangentMode=RCTM_User, time=0, value=0}
             {arriveTangent=0.5799, interpMode=RCIM_Cubic, leaveTangent=0.5799, tangentMode=RCTM_User, time=0.35, value=1}
            ]
           }
           YCurve {
            defaultValue=-0.1
            keys [
             {arriveTangent=0.0249, arriveTangentWeight=0.3333, interpMode=RCIM_Cubic, leaveTangent=0.0249, tangentMode=RCTM_User, time=0, value=0}
             {arriveTangent=0.5799, interpMode=RCIM_Cubic, leaveTangent=0.5799, tangentMode=RCTM_User, time=0.35, value=1}
            ]
           }
           ZCurve {
            defaultValue=1
            keys [
             {arriveTangent=0.0249, arriveTangentWeight=0.3333, interpMode=RCIM_Cubic, leaveTangent=0.0249, tangentMode=RCTM_User, time=0, value=0}
             {arriveTangent=0.5799, interpMode=RCIM_Cubic, leaveTangent=0.5799, tangentMode=RCTM_User, time=0.35, value=1}
            ]
           }
          }}
         Scale Curve=<4 3 1>
        )
       )
       Quaternion=$Particles.SkeletalMeshLocation001.SampledRotation
      )
      Alpha=&FloatFromCurve(
       FloatCurve=DataInterface() {propertyValues {Curve {keys [
           {arriveTangent=6.2442, interpMode=RCIM_Cubic, leaveTangent=6.2442, tangentMode=RCTM_User, time=0, value=0.25}
           {arriveTangent=0.4524, interpMode=RCIM_Cubic, leaveTangent=0.4524, tangentMode=RCTM_User, time=0.2, value=1}
           {arriveTangent=-3.2379, interpMode=RCIM_Cubic, leaveTangent=-3.2379, tangentMode=RCTM_User, time=0.8, value=0}
          ]}}}
       Scale Curve=&MergeHelper_FloatSwitch(0=0.75, 1=0.5, 2=?, 3=?, Index=&SpawnGroup(), Num=::NewEnumerator0)
      )
     )
    )) {isSetParameters=true}
   &RotateAroundPoint(
    Derive the Rotation Axis=false
    Radius:NiagaraFloat=&Multiply_Float(
     A=&Lerp_Float(
      A=&RandomRangeFloat(Minimum=0.5, Maximum=1)
      B=&RandomRangeFloat(Minimum=2, Maximum=3)
      Alpha=&FloatFromCurve(FloatCurve=DataInterface() {propertyValues {Curve {keys [
           {arriveTangent=3.8353, interpMode=RCIM_Cubic, leaveTangent=3.8353, tangentMode=RCTM_User, time=0, value=0.5}
           {arriveTangent=-0.5, interpMode=RCIM_Cubic, leaveTangent=-0.5, time=0.3, value=0.9847}
           {interpMode=RCIM_Cubic, leaveTangent=-1.7958, time=1, value=0}
          ]}}})
     )
     B=&SelectFloatFromArray(Direct Array Index=$Particles.SpawnGroup, Float Selection Array=DataInterface() {propertyValues {FloatData [1,0.75]}})
    )
    Rotation Axis:Vector3f=$System.DiveVelocityScale
    Rotation Phase:NiagaraFloat=&RandomRangeFloat(Minimum=0, Maximum=1)
    Rotation Rate:NiagaraFloat=&FloatFromCurve(
     FloatCurve=DataInterface() {propertyValues {Curve {defaultValue=1, keys [{tangentMode=RCTM_User, time=0.3, value=0}, {tangentMode=RCTM_User, time=0.8, value=1}]}}}
     Scale Curve=&MakeCustomFloatFromBool(Bool=&RandomBool(), True Float=0.5, False Float=-0.5)
    )
    Set Particle Position=false
   )
   /SetVariables(Particles.Rotational_Position:NiagaraPosition=&ConvertVectorToPosition(
     Input Position=&Add_Vector(
      A=&ConvertPositionToVector(Input Position=$Particles.Position, Convert To Absolute Large World Vector=false)
      B=Linked() {linkedVariable {name=Output.RotateAroundPoint.OutputPosition, type=NiagaraPosition}}
     )
     Vector Is In Absolute Large World Space:NiagaraBool=true
    )) {isSetParameters=true}
   &DynamicMaterialParameters(
    Index 0 Param 1=&FloatFromCurve(
     FloatCurve=DataInterface() {propertyValues {Curve {
        defaultValue=1
        keys [
         {arriveTangent=0.0618, interpMode=RCIM_Cubic, leaveTangent=0.0618, tangentMode=RCTM_User, time=0, value=0}
         <at 0.3 0.1125 RCIM_Cubic>
         {interpMode=RCIM_Cubic, tangentMode=RCTM_User, time=1, value=0}
        ]
       }}}
     Scale Curve=0.5
    )
    Index 0 Param 3=&RandomRangeFloat(Minimum=0.75, Maximum=1.75)
    Index 0 Param 4=1
    Index 1 Param 1=&FN_FrontEndFloatModifier(Gameplay=0.3, Front End=0.7)
    Module.Param0WriteEnabled=true
    Module.Param0WriteEnabled1=true
    Module.Param1WriteEnabled=true
    Module.Param1WriteEnabled1=true
    Module.Param2WriteEnabled=true
    Module.Param2WriteEnabled1=true
    Module.Param3WriteEnabled=true
    Module.Param3WriteEnabled1=true
    Write Parameter Index 1=true
   )
   &UpdateMeshOrientation(
    Facing Direction=&Add_Vector(
     A=$System.DiveVelocityScale
     B=&RotateVectorByQuaternion(
      Vector To Rotate=&VectorFromCurve(
       VectorCurve=DataInterface() {propertyValues {
         XCurve {
          defaultValue=0
          keys [
           {arriveTangent=2.7637, interpMode=RCIM_Cubic, time=0.25, value=1}
           {arriveTangent=-3.6109, interpMode=RCIM_Cubic, leaveTangent=-3.6109, tangentMode=RCTM_User, time=1, value=-0.4}
          ]
         }
         YCurve {keys [
           <at 0 1>
           {arriveTangent=-0.4977, interpMode=RCIM_Cubic, leaveTangent=-0.4977, tangentMode=RCTM_User, time=0.25, value=1}
           {arriveTangent=-3.8093, arriveTangentWeight=0.25, interpMode=RCIM_Cubic, leaveTangent=-3.8093, tangentMode=RCTM_User, time=1, value=-0.4}
          ]}
         ZCurve {keys [
           <at 0 1>
           {arriveTangent=-0.4977, interpMode=RCIM_Cubic, leaveTangent=-0.4977, tangentMode=RCTM_User, time=0.25, value=1}
           {arriveTangent=-3.8093, arriveTangentWeight=0.25, interpMode=RCIM_Cubic, leaveTangent=-3.8093, tangentMode=RCTM_User, time=1, value=-0.4}
          ]}
        }}
       Scale Curve=<1.2 -0.1 0>
      )
      Quaternion=$Particles.SkeletalMeshLocation001.SampledRotation
     )
    )
    Facing Interpolation=false
    Orientation Method=::NewEnumerator1
    Override Quat=false
    Side Direction:Vector3f=$Particles.RotateAroundPoint.VectorFromCenter
    Stabilization Mode:ENiagara_OrientVectorCount=::NewEnumerator2
    TransformFacing=false
    TransformSide=false
   )
   &ScaleMeshSize(Scale Factor=&Multiply_VectorByFloat(
     Vector=&Lerp_Vector(
      A=<0.5 1 1>
      B=<1 1 1>
      Alpha=&FloatFromCurve(FloatCurve=DataInterface() {propertyValues {Curve {keys [
           {arriveTangent=0.9666, interpMode=RCIM_Cubic, time=0.2, value=0}
           {interpMode=RCIM_Cubic, tangentMode=RCTM_User, time=0.4, value=1}
           {interpMode=RCIM_Cubic, tangentMode=RCTM_User, time=0.9, value=1}
           {arriveTangent=-17.1731, interpMode=RCIM_Cubic, leaveTangent=-17.1731, tangentMode=RCTM_User, time=1, value=0}
          ]}}})
     )
     Float=&FloatFromCurve(
      FloatCurve=DataInterface() {propertyValues {Curve {keys [
          {arriveTangent=3.3011, interpMode=RCIM_Cubic, leaveTangent=3.3011, tangentMode=RCTM_User, time=0, value=0.25}
          {interpMode=RCIM_Cubic, tangentMode=RCTM_User, time=0.15, value=1}
          {interpMode=RCIM_Cubic, tangentMode=RCTM_Break, time=0.5, value=1}
          {arriveTangent=-3.8115, interpMode=RCIM_Constant, leaveTangent=-3.8115, time=1, value=0.25}
         ]}}}
      Scale Curve=&Lerp_Float(A=1, B=2.2, Alpha=$NPC.FortniteNPC.BigHeadMode)
     )
    ))
   &ScaleSpriteSize(
    Non-Uniform Curve Scale=&Lerp_Vector2(A=<1 1.75>, B=<2 3.5>, Alpha=$NPC.FortniteNPC.BigHeadMode)
    Non-Uniform Curve Sprite Scale=DataInterface() {propertyValues {
      XCurve {
       defaultValue=0
       keys [
        <at 0 0 RCIM_Cubic>
        {arriveTangent=1.1429, interpMode=RCIM_Cubic, leaveTangent=1.1429, time=0.2487, value=1.1246}
        {arriveTangent=-1.497, interpMode=RCIM_Cubic, leaveTangent=-1.497, time=0.7, value=0.8}
        {arriveTangent=-3.0554, arriveTangentWeight=0.1, interpMode=RCIM_Cubic, leaveTangent=-3.0554, tangentMode=RCTM_User, time=1, value=0}
       ]
      }
      YCurve {
       defaultValue=0
       keys [
        <at 0 0 RCIM_Cubic>
        {arriveTangent=1.1429, interpMode=RCIM_Cubic, leaveTangent=1.1429, time=0.2487, value=1.1246}
        {arriveTangent=-1.497, interpMode=RCIM_Cubic, leaveTangent=-1.497, time=0.7, value=0.8}
        {arriveTangent=-3.0554, arriveTangentWeight=0.1, interpMode=RCIM_Cubic, leaveTangent=-3.0554, tangentMode=RCTM_User, time=1, value=0}
       ]
      }
     }}
    Scale Sprite Size Mode=::NewEnumerator3
   )
   &ScaleColor(
    Scale Alpha=&MakeCustomFloatFromBool(Bool=$User.bisInVehicle, True Float=0, False Float=1)
    ScaleA=true
    ScaleRGB=true
   ) {module=ScaleColor001}
   &ScaleColor(
    Scale Alpha=&FloatFromCurve(FloatCurve=DataInterface() {propertyValues {Curve {
        defaultValue=0
        keys [
         {interpMode=RCIM_Cubic, tangentMode=RCTM_User, time=0, value=0}
         {arriveTangent=0.3339, interpMode=RCIM_Cubic, leaveTangent=0.3339, tangentMode=RCTM_User, time=0.1, value=0.9}
         {interpMode=RCIM_Cubic, tangentMode=RCTM_User, time=0.7, value=0.9}
         {tangentMode=RCTM_User, time=1, value=0}
        ]
       }}})
    Scale RGB=&VectorFromFloat(Value=&FNTimeOfDayScalar(Night Scale=0.6, Morning Scale=1, Evening Scale=1))
    ScaleA=true
    ScaleRGB=true
   ) {module=ScaleColor002}
   /SetVariables(Particles.VisibilityTag=&MakeCustomIntFromBool(
     Bool=&Boolean_OrOperation(
      Bool1=$User.bisHeadHidden
      Bool2=&SetBoolByFloatComparison(A=$User.MaskOff, B=1, Comparison Type=::NewEnumerator4)
     )
     True INT=1
     False INT=0
    )) {isSetParameters=true, module=SetVariables_#5}
   /SetVariables(Particles.Color_Glow:LinearColor=&MultiplyLinearColor(
     Linear Color=<rgba 0.0681 0.7611 1 0.05>
     Scale Factor=&MakeVector4_FromLinearColor(LinearColor=$Particles.Color)
    )) {isSetParameters=true, module=SetVariables_#6}
  ]
  rendererClasses [&NiagaraMeshRendererProperties, &NiagaraSpriteRendererProperties]
  renderers [
   {rendererClass=&NiagaraMeshRendererProperties, rendererIndex=0}
   {rendererClass=&NiagaraSpriteRendererProperties, rendererIndex=1}
  ]
 }
 Derez_Mesh {
  emitterUpdateScript [
   &EmitterState(
    Inactive Response=::NewEnumerator1
    Life Cycle Mode=::NewEnumerator1
    Loop Behavior=::NewEnumerator1
    Loop Duration Mode=::NewEnumerator1
    UseLoopDelay=false
   )
   &SpawnBurst_Instantaneous(Use Spawn Probability=false, UseLoopCountLimit=false)
  ]
  particleSpawnScript [
   &InitializeParticle(
    Lifetime=&MergeHelper_FloatSwitch(
     0=1
     1=&Lerp_Float(
      A=&RandomRangeFloat(Minimum=0.9, Maximum=1.1)
      B=5
      Alpha=&Divide_Float(A=&VectorLength(Vector=$Engine.Owner.Velocity), B=5000)
     )
     2=?
     3=?
     Index=&SpawnGroup()
     Num=::NewEnumerator0
    )
    Mesh Scale Mode=::NewEnumerator3
    UsePositionOffset=false
    Write Material Random=true
   )
   &InitialMeshOrientation(Mesh Orientation Mode=::NewEnumerator2, Rotation=<0 0.25 0>, Rotation Coordinate Space=::NewEnumerator1, Use Rotation Vector=true)
  ]
  particleUpdateScript [
   &ParticleState(Kill Particles When Lifetime Has Elapsed=false, Loop Particles Lifetime=true)
   &ScaleMeshSize(Scale Factor=&VectorFromFloat(Value=&Lerp_Float(A=1, B=2, Alpha=$NPC.FortniteNPC.BigHeadMode)))
   &SkeletalMeshLocation(
    Bone Sampling Mode=::NewEnumerator4
    MaskSpawnGroup=false
    Sampled Position Offset=&Lerp_Vector(A=<-0.1 1.735 0>, B=<-0.2 3.47 0>, Alpha=$NPC.FortniteNPC.BigHeadMode)
    Skeletal Mesh=$System.SkeletalMesh
   )
   /SetVariables(Particles.VisibilityTag=&Multiply_Int(
     A=&MakeIntFromFloat(Float=&MakeCustomFloatFromBool(
       Bool=&SetBoolByFloatComparison(A=$User.MaskOff, B=0, Comparison Type=::NewEnumerator4)
       True Float=1
       False Float=0
      ))
     B=&MakeCustomIntFromBool(Bool=$User.bisHeadHidden, True INT=10, False INT=1)
    )) {isSetParameters=true, module=SetVariables_#5}
  ]
  rendererClasses [&NiagaraMeshRendererProperties]
  renderers [{rendererClass=&NiagaraMeshRendererProperties, rendererIndex=0}]
 }
}
summary {
 systemName=NS_DoubleConk_Mask_Idle
 emitters [
  {emitterName=Derez_Mesh, rendererClasses [&NiagaraMeshRendererProperties]}
  {
   emitterName=Derez_Cubes_Mask
   rendererClasses [&NiagaraMeshRendererProperties, &NiagaraSpriteRendererProperties]
  }
  {
   emitterName=Derez_Cube_Head
   rendererClasses [&NiagaraMeshRendererProperties, &NiagaraSpriteRendererProperties]
  }
 ]
 userVariables [
  User.bIsFrontEnd {type=NiagaraBool}
  User.bIsFrontEndPreview {type=NiagaraBool}
  User.bisHeadHidden {type=NiagaraBool}
  User.bisInVehicle {type=NiagaraBool}
  User.IsSkydiving {type=NiagaraBool}
  User.MaskOff {type=NiagaraFloat, defaultValue=0, description=""}
  User.Style {type=NiagaraFloat, defaultValue=0, description=""}
 ]
}
