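# Material function for the DoubleConk character's animated grid effect, modifying incoming material attributes with an added grid normal, emissive glow and optional opacity cutouts.
# Grid normal values give cells varied pulse timing, combined with grid lines from the emissive texture and a noisy band sweeping through local or bounds space.
# Static switches add backpack behavior: time-stepped sphere-mask bands along rotated local position cut opacity, with two-sided shading support.
# An interior-cubemap lookup builds fake backface color and masking.
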MaterialFunction /BRCosmetics/Characters/Player/Male/Medium/Bodies/M_MED_DoubleConk/Materials/MF_DoubleConkGrid
graph {nodes [
  n1 = TextureCoordinate()
  n2 = MaterialFunctionCall(Float2=@n1) {materialFunction=&BreakOutFloat2Components}
  Add(A=@n2.R, B=Constant() {r=1})
  Constant2Vector() {g=80, r=80}
  Constant()
  Constant() {r=0.64}
  Distance()
  FunctionOutput(None=BlendMaterialAttributes(A=NamedRerouteUsage(), B=NamedRerouteUsage(), Alpha=NamedRerouteUsage()))
  n3 = ScalarParameter() {defaultValue=0.5565, parameterName=BP_RotationAmout}
  n4 = Constant()
  n5 = LocalPosition() {localOrigin=InstancePreSkinning}
  n6 = VertexInterpolator(VS=MaterialFunctionCall(Float3=MaterialFunctionCall!(Rotation Amount=@n3, PivotPoint=@n4, WorldPosition=@n5.XYZ) {materialFunction=/Engine/Functions/Engine_MaterialFunctions02/WorldPositionOffset/RotateAboutWorldAxis_cheap}.Y-Axis) {materialFunction=&BreakOutFloat3Components}.B)
  n7 = Divide(A=Add(A=ShaderStageSwitch(PixelShader=@n6.PS, VertexShader=@n5.Z), B=Constant() {r=-10})) {constB=35}
  n8 = Time()
  n9 = ScalarParameter() {defaultValue=0.5, parameterName=BPBlankSpeed}
  n10 = Power(Base=Frac(None=Multiply(A=@n8, B=@n9) {constB=-0.2})) {constExponent=1}
  n11 = ScalarParameter() {defaultValue=15, parameterName=BPGridBlankDivider}
  n12 = Divide(A=Ceil(None=Multiply(A=@n10, B=@n11) {constB=10}), B=@n11)
  n13 = ScalarParameter() {defaultValue=0.035, parameterName=GridWdth}
  n14 = SphereMask(A=@n7, B=Add(A=@n12, B=ScalarParameter() {defaultValue=-0.031, parameterName=BPBlankGridOffset}) {constB=0.01}, Radius=@n13) {attenuationRadius=0.01}
  n15 = NamedRerouteUsage()
  n16 = NamedRerouteUsage()
  n17 = SphereMask(A=@n7, B=Add(A=@n12, B=ScalarParameter() {defaultValue=-0.29, parameterName=BPBlankGridOffset2}) {constB=0.01}, Radius=@n13) {attenuationRadius=0.01}
  n18 = Add(A=@n12, B=ScalarParameter() {defaultValue=-0.154, parameterName=BPBlankGridOffset3}) {constB=0.01}
  n19 = Add(
   A=Add(A=Multiply(A=@n14, B=@n15), B=Multiply(A=@n16, B=@n17))
   B=Multiply(A=SphereMask(A=@n7, B=@n18, Radius=@n13) {attenuationRadius=0.01}, B=NamedRerouteUsage())
  )
  LinearInterpolate(A=@n19)
  n20 = OneMinus(None=@n2.G)
  n21 = TextureSample(UVs=AppendVector(A=@n2.R, B=Power(Base=@n2.G, Exp=@n20))) {
   samplerType=SAMPLERTYPE_Normal
   texture=/BRCosmetics/Accessories/FORT_Backpacks/Backpack_F_MED_SpatialTravel/Textures/T_SpatialTravel_Backpack_N
  }
  LinearInterpolate(
   A=SphereMask(
    A=MaterialFunctionCall(Float2=@n1) {materialFunction=&BreakOutFloat2Components}.G
    B=Constant() {r=0.76}
   ) {attenuationRadius=0.01}
   B=@n21.RGB
  )
  n22 = MaterialFunctionCall() {materialFunction=&BoundingBoxBased_0-1_UVW}
  MaterialFunctionCall(X=@n22.R, Y=@n22.R) {materialFunction=&MakeFloat2}
  n23 = MaterialFunctionCall(UVs=TextureCoordinate(), Rotation Center=Constant(), "Rotation Angle (0-1)"=Constant() {r=-0.09}) {materialFunction=&CustomRotator}
  n24 = MaterialFunctionCall(Float2=@n23.Rotated Values) {materialFunction=&BreakOutFloat2Components}
  n25 = Divide(A=Multiply(A=@n24.G) {constB=-1}) {constB=50}
  MaterialFunctionCall(X=Power(Base=Abs(None=@n24.R), Exp=Add(A=@n25)), Y=@n24.G) {materialFunction=&MakeFloat2}
  n26 = Add(A=Multiply(A=TextureCoordinate()) {constB=0.35}, B=Constant2Vector() {g=0.805, r=0.77})
  n27 = MaterialFunctionCall(Float2=@n26) {materialFunction=&BreakOutFloat2Components}
  n28 = Power(Base=@n27.R) {constExponent=5}
  MaterialFunctionCall(X=@n27.R, Y=Power(Base=Abs(None=@n27.G), Exp=Add(A=@n28))) {materialFunction=&MakeFloat2}
  n29 = ScalarParameter() {defaultValue=80, parameterName=CubeMapTiling}
  n30 = TextureSample(UVs=Multiply(A=MaterialFunctionCall!(UVs=@n26, Tiling=@n29, Randomize Rotation) {materialFunction=/Engine/Functions/Engine_MaterialFunctions01/Cubemaps/InteriorCubemap}.UVW)) {samplerType=SAMPLERTYPE_LinearColor, texture=/Game/Global/Textures/Cubemaps/TC_InteriorNormals_Tex}
  n31 = Constant3Vector() {constant=<rgba 0 3 3 1>}
  Multiply(A=@n30.RGB, B=@n31)
  n32 = LocalPosition() {localOrigin=InstancePreSkinning}
  n33 = MaterialFunctionCall(Float3=ShaderStageSwitch(PixelShader=VertexInterpolator(VS=@n32.XYZ).PS, VertexShader=@n32.XYZ)) {materialFunction=&BreakOutFloat3Components}
  Multiply(A=@n33.R)
  n34 = MaterialFunctionCall(Float3=LocalPosition().XYZ) {materialFunction=&BreakOutFloat3Components}
  Multiply(
   A=MaterialFunctionCall(X=@n34.B, Y=@n34.B) {materialFunction=&MakeFloat2}.Result
   B=Constant2Vector() {g=-1, r=1}
  )
  Multiply(A=@n23.Rotated Values)
  Multiply() {constB=-1}
  Multiply(A=@n27.R) {constB=2}
  n35 = TextureSampleParameter2D() {
   group=STGridFX
   parameterName=AdditionalGrid
   samplerType=SAMPLERTYPE_Normal
   texture=/BRCosmetics/Characters/Player/Male/Medium/Bodies/M_MED_DoubleConk/Textures/T_DoubleConk_Body_N_Grid
  }
  n36 = ScalarParameter() {defaultValue=1, group=STGridFX, parameterName=AdditionalGridStrength}
  n37 = MaterialFunctionCall(Float3=MaterialFunctionCall(Normal=@n35.RGB, Flatness=OneMinus(None=@n36)) {materialFunction=&FlattenNormal}.Result) {materialFunction=&BreakOutFloat3Components}
  Multiply(A=@n37.R)
  n38 = Constant3Vector() {constant=<rgba 0.5 0.5 1 0>}
  NamedRerouteDeclaration(None=DotProduct(A=@n35.RGB, B=@n38)) {name=GridGrayscale, nodeColor=<rgba 0 0.3906 1 1>}
  n39 = GetMaterialAttributes(None=NamedRerouteDeclaration(None=FunctionInput() {inputName=MatIn, inputType=FunctionInput_MaterialAttributes}) {name=MatIn, nodeColor=<rgba 1 0.2812 0 1>}) {attributeGetTypes [#1, #2, #3]}
  NamedRerouteDeclaration(None=@n39.BaseColor) {name=BaseColor, nodeColor=<rgba 1 0.4219 0 1>}
  n40 = TextureSample() {samplerType=SAMPLERTYPE_Masks, texture=/BRCosmetics/Accessories/FORT_Backpacks/Backpack_F_MED_SpatialTravel/Textures/T_SpatialTravel_Anim}
  n41 = Constant() {r=0.1}
  n42 = Constant() {r=0.4}
  NamedRerouteDeclaration(None=Add(A=SphereMask(A=@n40.R, B=@n41) {attenuationRadius=0.02}, B=SphereMask(A=@n40.R, B=@n42) {attenuationRadius=0.02})) {name=Colume1, nodeColor=<rgba 1 0.8906 0 1>}
  n43 = SphereMask(A=@n40.R, B=Constant() {r=0.2}) {attenuationRadius=0.02}
  n44 = SphereMask(A=@n40.R, B=Constant() {r=0.5}) {attenuationRadius=0.02}
  n45 = Constant() {r=0.7}
  NamedRerouteDeclaration(None=Add(A=Add(A=@n43, B=@n44), B=SphereMask(A=@n40.R, B=@n45) {attenuationRadius=0.02})) {name=Colume2, nodeColor=<rgba 1 0.8906 0 1>}
  n46 = Constant() {r=0.3}
  n47 = Constant() {r=0.6}
  NamedRerouteDeclaration(None=Add(A=SphereMask(A=@n40.R, B=@n46) {attenuationRadius=0.02}, B=SphereMask(A=@n40.R, B=@n47) {attenuationRadius=0.02})) {name=Colume3, nodeColor=<rgba 1 0.8906 0 1>}
  n48 = SphereMask(A=@n30.RGB, B=Constant3Vector() {constant=<rgba 0.5 0.5 1 0>}) {attenuationRadius=0.2, hardnessPercent=0}
  n49 = Constant3Vector() {constant=<rgba 0 0.5 0.5 0>}
  n50 = SphereMask(A=@n30.RGB, B=@n49) {attenuationRadius=0.2, hardnessPercent=0}
  NamedRerouteDeclaration(None=Add(A=Multiply(A=@n48, B=@n31), B=Multiply(A=@n50, B=@n49))) {name=BackfaceColor, nodeColor=<rgba 0 1 0.1562 1>}
  n51 = Saturate(None=NamedRerouteDeclaration(None=@n48) {name=BackFace, nodeColor=<rgba 0 1 0.4609 1>})
  NamedRerouteDeclaration(None=Max(A=NamedRerouteUsage(), B=OneMinus(None=@n51))) {name=OpacityMask_Back, nodeColor=<rgba 1 0.9141 0 1>}
  n52 = Divide(A=AppendVector(A=Abs(None=@n33.R), B=@n33.B)) {constB=100}
  n53 = SphereMask(A=@n52, B=Constant2Vector() {g=0.95, r=0.55}) {attenuationRadius=0.4, hardnessPercent=20}
  n54 = SphereMask(A=@n52, B=Constant2Vector() {r=0.2}) {attenuationRadius=0.18, hardnessPercent=50}
  NamedRerouteDeclaration(None=Power(Base=Add(A=@n53, B=@n54)) {constExponent=0.8}) {name=PalmMask, nodeColor=<rgba 0 0.625 1 1>}
  n55 = VectorParameter() {defaultValue=<rgba 0.0171 0.0444 0.1094 1>}
  n56 = NamedRerouteUsage()
  n57 = NamedRerouteUsage()
  n58 = LinearInterpolate(A=@n55.RGB, B=@n56, Alpha=@n57) {constB=0}
  n59 = NamedRerouteUsage()
  n60 = LinearInterpolate(A=@n55.RGB, B=@n59, Alpha=@n57)
  n61 = SetMaterialAttributes!(
   MaterialAttributes=NamedRerouteUsage()
   Normal=NamedRerouteUsage()
   EmissiveColor=StaticSwitchParameter(True=StaticSwitchParameter(True=@n60, False=@n59) {group=STGridFX, parameterName=UseTwoSidedBackpack}, False=@n59) {group=STGridFX, parameterName=IsSTBackpack}
   OpacityMask=StaticSwitchParameter(True=NamedRerouteUsage(), False=Constant() {r=1}) {group=STGridFX, parameterName=IsSTBackpack}
   BaseColor=StaticSwitchParameter(True=StaticSwitchParameter(True=@n58, False=@n56) {group=STGridFX, parameterName=UseTwoSidedBackpack}, False=@n56) {group=STGridFX, parameterName=IsSTBackpack}
  ) {attributeSetTypes [
    #1
    #2
    #4
    #3
   ]}
  NamedRerouteDeclaration(None=@n61) {name=MatOut, nodeColor=<rgba 0 0.0391 1 1>}
  NamedRerouteDeclaration(None=ChannelMaskParameter(None=FunctionInput() {inputName=TechArtMask}) {group=STGridFX, parameterName=GridAnimChannel}.R) {name=GridAnimMask, nodeColor=<rgba 0.3594 1 0 1>}
  NamedRerouteDeclaration(None=@n39.Normal) {name=NormalIn, nodeColor=<rgba 1 0 0.0937 1>}
  n62 = ScalarParameter() {defaultValue=0.2, group=STGridFX, parameterName=AdditionalGridFlatness}
  NamedRerouteDeclaration(None=MaterialFunctionCall(
    BaseNormal=NamedRerouteUsage()
    AdditionalNormal=MaterialFunctionCall(Normal=@n35.RGB, Flatness=@n62) {materialFunction=&FlattenNormal}.Result
   ) {materialFunction=&BlendAngleCorrectedNormals}.Result) {name=NormalOut, nodeColor=<rgba 0 0.4141 1 1>}
  n63 = ScalarParameter() {defaultValue=0.7, group=STGridFX, parameterName=GridEmissiveMult_High}
  n64 = ScalarParameter() {defaultValue=0.4, group=STGridFX, parameterName=GridEmissiveMult_Low}
  n65 = DotProduct(
   A=MaterialFunctionCall(X=@n37.G, Y=@n37.G, Z=@n37.G) {materialFunction=&MakeFloat3}.Result
   B=@n38
  )
  n66 = Add(A=@n65, B=Time())
  n67 = ScalarParameter() {defaultValue=0.5, group=STGridFX, parameterName=GridAnimSpeed}
  n68 = Frac(None=Add(A=DotProduct(A=@n35.RGB, B=@n38), B=Multiply(A=@n66, B=@n67)))
  n69 = Saturate(None=ConstantBiasScale(None=Sine(None=@n68)))
  n70 = TextureSampleParameter2D() {
   group=06_Emissive
   parameterName=Emissive
   samplerSource=SSM_Wrap_WorldGroupSettings
   texture=/BRCosmetics/Characters/Player/Male/Medium/Bodies/M_MED_DoubleConk/Textures/T_DoubleConk_Body_E
  }
  n71 = VectorParameter() {group=STGridFX, parameterName=GridLineTargetColor, defaultValue=<rgba 0 0.5962 1 1>}
  n72 = Constant3Vector() {constant=<rgba 0 0 1 0>}
  n73 = Max(
   A=NamedRerouteDeclaration(None=SphereMask(A=@n70.RGB, B=@n71.RGB) {attenuationRadius=0.63}) {name=Grid Line, nodeColor=<rgba 0.8281 0 1 1>}
   B=Power(Base=DotProduct(A=@n35.RGB, B=@n72)) {constExponent=500}
  )
  n74 = NamedRerouteUsage()
  n75 = Multiply(
   A=LinearInterpolate(A=LinearInterpolate(A=@n63, B=@n64, Alpha=@n69), B=NamedRerouteUsage(), Alpha=Max(A=@n73, B=@n74)) {constB=0}
   B=NamedRerouteUsage()
  )
  n76 = ScalarParameter() {defaultValue=100, group=STGridFX, parameterName=BandSpreadingLocalPositionDivider}
  n77 = SphereMask(
   A=Divide(A=AppendVector(A=@n33.R, B=@n33.R), B=@n76) {constB=100}
   B=AppendVector(A=ScalarParameter() {parameterName=BandCenterPositionX}, B=ScalarParameter() {defaultValue=1.3, parameterName=BandCenterPositionZ})
  ) {attenuationRadius=2, hardnessPercent=0}
  n78 = OneMinus(None=SphereMask(
    A=MaterialFunctionCall(X=@n22.R, Y=@n22.R) {materialFunction=&MakeFloat2}.Result
    B=Constant() {r=0.5}
   ) {attenuationRadius=0.5, hardnessPercent=0})
  n79 = AppendVector(
   A=ScalarParameter() {group=STGridFX, parameterName=BandLocalPositionOffsetX}
   B=ScalarParameter() {group=STGridFX, parameterName=BandLocalPositionOffsety}
  )
  n80 = SphereMask(A=Divide(A=Add(A=@n32.XY, B=@n79), B=@n76) {constB=100}, B=Constant()) {attenuationRadius=1.5, hardnessPercent=0}
  n81 = ScalarParameter() {group=STGridFX, parameterName=BandSpreadingUseLocalPosition}
  n82 = Add(
   A=StaticSwitchParameter(True=OneMinus(None=@n77), False=LinearInterpolate(A=@n78, B=@n80, Alpha=@n81)) {defaultValue=true, group=STGridFX, parameterName=UseLocalXZforBand}
   B=Multiply(A=NamedRerouteUsage(), B=ScalarParameter() {defaultValue=0.05, group=STGridFX, parameterName=BandSpreadingNoiseAmount}) {constB=0.1}
  )
  n83 = Multiply(A=Time(), B=ScalarParameter() {defaultValue=0.35, group=STGridFX, parameterName=BandSpreadingSpeed})
  n84 = Multiply(A=Power(Base=Frac(None=@n83)), B=ScalarParameter() {defaultValue=1.3, group=STGridFX, parameterName=BandLocationMAX}) {constB=1.1}
  n85 = ScalarParameter() {defaultValue=0.15, group=STGridFX, parameterName=BandWidth}
  n86 = Multiply(
   A=Power(Base=SphereMask(A=@n82, B=@n84, Radius=@n85) {attenuationRadius=0.1, hardnessPercent=0.5}) {constExponent=1}
   B=ScalarParameter() {defaultValue=2, group=STGridFX, parameterName=BandEmissiveMult}
  )
  n87 = NamedRerouteUsage()
  n88 = NamedRerouteUsage()
  n89 = Multiply(A=LinearInterpolate(A=LinearInterpolate(A=@n86, Alpha=@n87) {constB=0}, Alpha=@n88) {constA=1, constB=0}, B=NamedRerouteUsage())
  n90 = SphereMask(A=NamedRerouteUsage(), B=Constant()) {attenuationRadius=0.1, hardnessPercent=0}
  NamedRerouteDeclaration(None=Multiply(A=Add(A=@n75, B=@n89), B=OneMinus(None=@n90))) {name=EmissiveOut, nodeColor=<rgba 0.3359 1 0 1>}
  NamedRerouteDeclaration(None=@n39.EmissiveColor) {name=EmissiveIn, nodeColor=<rgba 0 0.5781 1 1>}
  NamedRerouteDeclaration(None=Power(Base=@n65, Exp=ScalarParameter() {defaultValue=2, group=STGridFX, parameterName=GridRandomMaskPower}) {constExponent=3}) {name=GridRandomMask, nodeColor=<rgba 0.4531 1 0 1>}
  n91 = Step(Y=@n7) {constX=0.22}
  n92 = OneMinus(None=Multiply(A=OneMinus(None=@n91), B=@n19))
  n93 = TwoSidedSign()
  NamedRerouteDeclaration(None=StaticSwitchParameter(True=LinearInterpolate(B=@n92, Alpha=@n93) {constA=1}, False=@n92) {group=STGridFX, parameterName=UseTwoSidedBackpack}) {name=OpacityMask_Front, nodeColor=<rgba 0.7109 1 0 1>}
  NamedRerouteUsage()
  n94 = NamedRerouteUsage()
  OneMinus(None=Saturate(None=Sign(None=@n94)))
  OneMinus(None=Step(Y=@n94) {constX=0.5})
  Power(Base=@n69, Exp=Saturate(None=Add(A=@n65) {constB=-0.5}))
  Rotator()
  Sign(None=@n88)
  TextureCoordinate()
  TextureSampleParameter2D() {
   group=06_Emissive
   parameterName=Emissive
   samplerSource=SSM_Wrap_WorldGroupSettings
   texture=&T_DoubleConk_Body_E
  }
  TextureSample(UVs=@n26) {
   samplerType=SAMPLERTYPE_Normal
   texture=&T_SpatialTravel_Backpack_N
  }
  TextureSample(UVs=MaterialFunctionCall() {materialFunction=&InteriorCubemap}.UVW) {samplerType=SAMPLERTYPE_LinearColor, texture=&TC_InteriorNormals_Tex}
 ]}
