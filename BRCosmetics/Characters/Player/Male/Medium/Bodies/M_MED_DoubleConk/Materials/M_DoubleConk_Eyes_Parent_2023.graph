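# Parent eye material for the DoubleConk male medium character body, built for skeletal meshes and mesh deformers.
# It uses material attributes, starting from the shared MF_eyes_2023 eye function and passing the result through MF_TechArtOverrides.
# MF_DoubleConkGrid then applies the character-specific layer last, with its TechArtMask input fixed at full by a constant.
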Material /BRCosmetics/Characters/Player/Male/Medium/Bodies/M_MED_DoubleConk/Materials/M_DoubleConk_Eyes_Parent_2023
graph {
 properties {
  bUseMaterialAttributes=true
  bUsedWithMeshDeformer=true
  bUsedWithSkeletalMesh=true
  pixelDepthOffsetMode=PDOM_Legacy
  shadingModel=MSM_FromMaterialExpression
  usedShadingModels=Default Lit
 }
 nodes [MaterialFunctionCall!(
   MatIn=MaterialFunctionCall(In=MaterialFunctionCall() {materialFunction=/Game/Global/Materials/Character_Parent_2023/Eye_Parent_2023/MF_eyes_2023}.Result) {materialFunction=&MF_TechArtOverrides}.Result
   TechArtMask=Constant() {r=1}
  ) {materialFunction=&/MF_DoubleConkGrid}]
}
