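# Post-attach animation Blueprint for the M_MED_DoubleConk male body mesh, which follows the pose of its attached parent mesh.
# Copy Pose From Mesh feeds the parent's pose into the M_MED_DoubleConk_CtrlRig Control Rig, whose result is the output pose.
# The rig receives parachute-open, skydiving and emote-playing flags as inputs.
# The event graph holds only an unconnected update event, so all work happens in the AnimGraph.
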AnimBlueprint /BRCosmetics/Characters/Player/Male/Medium/Bodies/M_MED_DoubleConk/Meshes/M_MED_DoubleConk_AnimBP
target_skeleton=&/M_MED_DoubleConk_Skeleton
graphs [
 AnimGraph {nodes [
   n1 = Control Rig!(
    Source=Copy Pose From Mesh() {bUseAttachedParent=true}
    Alpha
    isGrabbingParachute_CR:bool=Get bIsParachuteOpened()
    isSkydiving_CR:bool=Get bIsSkyDiving()
    isEmoting_CR:bool=Get bIsPlayingEmote()
   ) -> (Pose) {
    alphaBoolBlend {blendOption=Linear}
    controlRigAssetReference {blueprintRigClass=&/M_MED_DoubleConk_CtrlRig.M_MED_DoubleConk_CtrlRig_C}
   }
   Output Pose(Result=@n1) {layerGroup=None}
  ]}
 EventGraph {nodes [Event BlueprintUpdateAnimation()]}
]
variables []
