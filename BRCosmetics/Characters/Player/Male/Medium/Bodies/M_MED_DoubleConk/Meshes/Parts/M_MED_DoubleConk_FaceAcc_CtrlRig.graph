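# Control Rig for the hair parts of the M_MED_DoubleConk character's face accessory mesh, adjusting back-hair bones as the head moves.
# The forward solve first applies fixed offsets to hair bones, then runs a two-branch sequence.
# Spherical pose readers measure head direction relative to neck_01 or spine_03 and weight Modify Transforms corrective poses on base and dynamic hair bones.
# Preparation imports the hierarchy from the skeleton.
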ControlRigBlueprint /BRCosmetics/Characters/Player/Male/Medium/Bodies/M_MED_DoubleConk/Meshes/Parts/M_MED_DoubleConk_FaceAcc_CtrlRig
graphs [
 RigVMFunctionLibrary {links [], nodes []}
 RigVMModel {nodes [
   n1 = &RigUnit_PrepareForExecution()
   &RigUnit_HierarchyImportFromSkeleton(NameSpace=None, bIncludeCurves=False, bIncludeMeshSockets=false, bIncludeVirtualBones=false, ExecutePin=@n1.ExecutePin)
   RigUnit_BeginExecution = &RigUnit_BeginExecution()
   n2 = &RigUnit_ModifyTransforms(
    ItemToModify="((Item=(Type=Bone,Name=\"dyn_hairBack_a_04\"),Transform=(Rotation=(X=0,Y=0,Z=-0.0872,W=0.9962),Translation=(X=0,Y=0,Z=0),Scale3D=(X=1,Y=1,Z=1))),(Item=(Type=Bone,Name=\"dyn_hairBack_b_04\"),Transform=(Rotation=(X=0,Y=0,Z=-0.1305,W=0.9914),Translation=(X=0,Y=0, ...(truncated)"
    Weight=1
    WeightMinimum=0
    ExecutePin=@RigUnit_BeginExecution.ExecutePin
   )
   RigVMFunction_Sequence = &RigVMFunction_Sequence(ExecuteContext=@n2.ExecutePin)
   n3 = &RigUnit_SphericalPoseReader(
    ActiveRegionScaleFactors=(PositiveWidth=1.55,NegativeWidth=1,PositiveHeight=1.3,NegativeHeight=1.3)
    ActiveRegionSize=0.1
    Debug=(bDrawDebug=false,bDraw2D=false,bDrawLocalAxes=false,DebugScale=25,DebugSegments=20,DebugThickness=0.25)
    DriverAxis=(X=1,Y=0,Z=0)
    DriverItem=(Type=Bone,Name="head")
    FalloffRegionScaleFactors=(PositiveWidth=2,NegativeWidth=1,PositiveHeight=1,NegativeHeight=1)
    FalloffSize=0.25
    FlipHeightScaling=False
    FlipWidthScaling=False
    OptionalParentItem=(Type=Bone,Name="neck_01")
    RotationOffset=(X=-90,Y=0,Z=20)
    ExecutePin=@RigVMFunction_Sequence.A
   )
   > &RigUnit_ModifyTransforms(
    ItemToModify="((Item=(Type=Bone,Name=\"base_hairBack_a\"),Transform=(Rotation=(X=0,Y=0,Z=-0.2164,W=0.9763),Translation=(X=-3,Y=0,Z=0),Scale3D=(X=1,Y=1,Z=1))),(Item=(Type=Bone,Name=\"base_hairBack_b\"),Transform=(Rotation=(X=0,Y=0,Z=-0.3173,W=0.9483),Translation=(X=-3,Y=0,Z= ...(truncated)"
    WeightMinimum=0
    Weight=@n3.OutputParam
   )
   > n4 = &RigUnit_SphericalPoseReader(
    ActiveRegionScaleFactors=(PositiveWidth=1.55,NegativeWidth=1,PositiveHeight=1.3,NegativeHeight=1.3)
    ActiveRegionSize=0.1
    Debug=(bDrawDebug=false,bDraw2D=false,bDrawLocalAxes=false,DebugScale=25,DebugSegments=20,DebugThickness=0.25)
    DriverAxis=(X=1,Y=0,Z=0)
    DriverItem=(Type=Bone,Name="head")
    FalloffRegionScaleFactors=(PositiveWidth=2,NegativeWidth=1,PositiveHeight=1,NegativeHeight=1)
    FalloffSize=0.25
    FlipHeightScaling=False
    FlipWidthScaling=False
    OptionalParentItem=(Type=Bone,Name="spine_03")
    RotationOffset=(X=-67.4652,Y=0,Z=17.262)
   )
   > &RigUnit_ModifyTransforms(
    ItemToModify=((Item=(Type=Bone,Name="dyn_hairBack_a_04"),Transform=(Rotation=(X=-0.5087,Y=-0.2163,Z=-0.2609,W=0.7914),Translation=(X=0,Y=0,Z=0),Scale3D=(X=1,Y=1,Z=1))))
    WeightMinimum=0
    Weight=@n4.OutputParam
   )
   n5 = &RigUnit_SphericalPoseReader(
    ActiveRegionScaleFactors=(PositiveWidth=1.4,NegativeWidth=1.6,PositiveHeight=1,NegativeHeight=1)
    ActiveRegionSize=0.1
    Debug=(bDrawDebug=false,bDraw2D=false,bDrawLocalAxes=false,DebugScale=25,DebugSegments=20,DebugThickness=0.25)
    DriverAxis=(X=1,Y=0,Z=0)
    DriverItem=(Type=Bone,Name="head")
    FalloffRegionScaleFactors=(PositiveWidth=1,NegativeWidth=1.6,PositiveHeight=1,NegativeHeight=1)
    FalloffSize=0.25
    FlipHeightScaling=False
    FlipWidthScaling=False
    OptionalParentItem=(Type=Bone,Name="neck_01")
    RotationOffset=(X=-90,Y=0,Z=124)
    ExecutePin=@RigVMFunction_Sequence.B
   )
   > &RigUnit_ModifyTransforms(
    ItemToModify="((Item=(Type=Bone,Name=\"base_hairBack_a\"),Transform=(Rotation=(X=0,Y=0,Z=-0.0698,W=0.9976),Translation=(X=0,Y=0,Z=0),Scale3D=(X=1,Y=1,Z=1))),(Item=(Type=Bone,Name=\"base_hairBack_b\"),Transform=(Rotation=(X=0,Y=0,Z=-0.0698,W=0.9976),Translation=(X=0,Y=0,Z=0) ...(truncated)"
    WeightMinimum=0
    Weight=@n5.OutputParam
   )
  ]}
]
