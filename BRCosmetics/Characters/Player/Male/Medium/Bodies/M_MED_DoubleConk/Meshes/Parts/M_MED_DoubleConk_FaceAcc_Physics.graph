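# Physics asset for the DoubleConk male medium character's face accessory mesh, providing physics for three back-hair chains.
# Each chain (a, b, c) links a small base body to three capsule segments through constraints with locked twist and narrow swing limits, tightest at the root.
# A head sphere and capsules on the neck, upper spine and clavicles form body shapes around the hair, with clavicles freely constrained to the spine.
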PhysicsAsset /BRCosmetics/Characters/Player/Male/Medium/Bodies/M_MED_DoubleConk/Meshes/Parts/M_MED_DoubleConk_FaceAcc_Physics
bodies {
 base_hairBack_a [{shapeType=Capsule, length=0.505, radius=0.505, shapeName=base_hairBack_a_capsule, rotation=<rot 0 -90 0>}]
 base_hairBack_b [{shapeType=Capsule, length=0.505, radius=0.505, shapeName=base_hairBack_b_capsule, rotation=<rot 0 -90 0>}]
 base_hairBack_c [{shapeType=Capsule, length=0.505, radius=0.505, shapeName=base_hairBack_c_capsule, rotation=<rot 0 -90 0>}]
 clavicle_l [{
   shapeType=Capsule
   length=11.3239
   radius=6.1991
   shapeName=clavicle_l_capsule
   center=<8.2479 0.0001 -0.2317>
   rotation=<rot -0.007 -89.9995 101.1675>
  }]
 clavicle_r [{
   shapeType=Capsule
   length=11.3239
   radius=6.1991
   shapeName=clavicle_l_capsule
   center=<-7.7878 -0.0001 -0.2831>
   rotation=<rot -0.0076 -89.9995 99.3235>
  }]
 dyn_hairBack_a_01 [{shapeType=Capsule, length=2.8259, radius=1.509, shapeName=dyn_hairBack_b_01_capsule, center=<3.3873 0 0>, rotation=<rot 0 0 -30>}]
 dyn_hairBack_a_02 [{shapeType=Capsule, length=2.8259, radius=1.509, shapeName=dyn_hairBack_b_01_capsule, center=<3.3873 -1 0>, rotation=<rot 0 0 -30>}]
 dyn_hairBack_a_03 [
  {shapeType=Capsule, length=2.8259, radius=1.509, shapeName=dyn_hairBack_b_01_capsule, center=<3.3873 -1 0>, rotation=<rot 0 0 -30>}
  {
   shapeType=Capsule
   length=2.1205
   radius=1.1835
   shapeName=dyn_hairBack_a_03_capsule
   center=<2.2164 -0.4442 -0.3749>
   rotation=<rot 1.8546 -3.326 -60.8903>
  }
 ]
 dyn_hairBack_b_01 [{shapeType=Capsule, length=2.8259, radius=1.509, shapeName=dyn_hairBack_b_01_capsule, center=<3.3873 0 0>}]
 dyn_hairBack_b_02 [{shapeType=Capsule, length=2.8259, radius=1.509, shapeName=dyn_hairBack_b_01_capsule, center=<3.3873 -1 0>}]
 dyn_hairBack_b_03 [{shapeType=Capsule, length=2.8259, radius=1.509, shapeName=dyn_hairBack_b_01_capsule, center=<3.3873 -1 0>}]
 dyn_hairBack_c_01 [{shapeType=Capsule, length=2.8259, radius=1.509, shapeName=dyn_hairBack_b_01_capsule, center=<3.3873 0 0>, rotation=<rot 0 0 30>}]
 dyn_hairBack_c_02 [{shapeType=Capsule, length=2.8259, radius=1.509, shapeName=dyn_hairBack_b_01_capsule, center=<3.3873 -1 0>, rotation=<rot 0 0 30>}]
 dyn_hairBack_c_03 [
  {shapeType=Capsule, length=2.8259, radius=1.509, shapeName=dyn_hairBack_b_01_capsule, center=<3.3873 -1 0>, rotation=<rot 0 0 30>}
  {
   shapeType=Capsule
   length=2.1205
   radius=1.1835
   shapeName=dyn_hairBack_a_03_capsule
   center=<2.7802 3.0728 1.3901>
   rotation=<rot 0 0.0002 -113.3538>
  }
 ]
 head [{shapeType=Sphere, radius=6.0012, shapeName=head_sphere, center=<1.6465 1.3822 0>}]
 neck_02 [{shapeType=Capsule, length=4.2305, radius=4.6482, shapeName=neck_02_capsule, center=<0 0.0595 0>, rotation=<rot 90 179.9998 -180>}]
 spine_05 [{shapeType=Capsule, length=7.9228, radius=7.5603, shapeName=spine_05_capsule, center=<7.1377 2.0765 -0.158>, rotation=<rot 0.7299 -90 -2.213>}]
}
constraints [
 {
  bone1Name=dyn_hairBack_a_02
  bone2Name=dyn_hairBack_a_01
  swing1LimitDegrees=25
  swing2LimitDegrees=10
  twistLimitDegrees=45
  twistMotion=Locked
 }
 {
  bone1Name=dyn_hairBack_a_03
  bone2Name=dyn_hairBack_a_02
  swing1LimitDegrees=25
  swing2LimitDegrees=14
  twistLimitDegrees=45
  twistMotion=Locked
 }
 {
  bone1Name=dyn_hairBack_b_02
  bone2Name=dyn_hairBack_b_01
  swing1LimitDegrees=25
  swing2LimitDegrees=10
  twistLimitDegrees=45
  twistMotion=Locked
 }
 {
  bone1Name=dyn_hairBack_b_03
  bone2Name=dyn_hairBack_b_02
  swing1LimitDegrees=25
  swing2LimitDegrees=14
  twistLimitDegrees=45
  twistMotion=Locked
 }
 {
  bone1Name=dyn_hairBack_c_02
  bone2Name=dyn_hairBack_c_01
  swing1LimitDegrees=25
  swing2LimitDegrees=10
  twistLimitDegrees=45
  twistMotion=Locked
 }
 {
  bone1Name=dyn_hairBack_c_03
  bone2Name=dyn_hairBack_c_02
  swing1LimitDegrees=25
  swing2LimitDegrees=14
  twistLimitDegrees=45
  twistMotion=Locked
 }
 {
  bone1Name=dyn_hairBack_a_01
  bone2Name=base_hairBack_a
  swing1LimitDegrees=17
  swing2LimitDegrees=6
  twistLimitDegrees=45
  twistMotion=Locked
 }
 {
  bone1Name=dyn_hairBack_b_01
  bone2Name=base_hairBack_b
  swing1LimitDegrees=17
  swing2LimitDegrees=3
  twistLimitDegrees=45
  twistMotion=Locked
 }
 {
  bone1Name=dyn_hairBack_c_01
  bone2Name=base_hairBack_c
  swing1LimitDegrees=17
  swing2LimitDegrees=6
  twistLimitDegrees=45
  twistMotion=Locked
 }
 {
  bone1Name=clavicle_r
  bone2Name=spine_05
  swing1LimitDegrees=45
  swing1Motion=Free
  swing2LimitDegrees=45
  swing2Motion=Free
  twistLimitDegrees=45
  twistMotion=Free
 }
 {
  bone1Name=clavicle_l
  bone2Name=spine_05
  swing1LimitDegrees=45
  swing1Motion=Free
  swing2LimitDegrees=45
  swing2Motion=Free
  twistLimitDegrees=45
  twistMotion=Free
 }
]
