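# Post-body animation blueprint for the DoubleConk male medium character's FaceAcc mesh, adding secondary motion to its hair bones.
# The attached parent's pose runs through a Control Rig, then a head-space RigidBody simulation.
# Chained single-bone AnimDynamics nodes add damped, limited angular springs to strand, front and back hair.
# Finally, bIsSkyDiving, not declared here, enables additive offsets on three back-hair root bones; the update event is empty.
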AnimBlueprint /BRCosmetics/Characters/Player/Male/Medium/Bodies/M_MED_DoubleConk/Meshes/Parts/M_MED_DoubleConk_FaceAcc_AnimBP
target_skeleton=&/M_MED_DoubleConk_FaceAcc_Skeleton
graphs [
 AnimGraph {nodes [
   n1 = Control Rig(Source=Copy Pose From Mesh() {bUseAttachedParent=true}) {
    alphaBoolBlend {blendOption=Linear}
    controlRigAssetReference {blueprintRigClass=&/M_MED_DoubleConk_FaceAcc_CtrlRig.M_MED_DoubleConk_FaceAcc_CtrlRig_C}
   }
   n2 = RigidBody!(ComponentPose=Local To Component(LocalPose=@n1), Alpha, OverrideWorldGravity:Vector="0,0,0", ExternalForce) -> (Pose) {
    overridePhysicsAsset=&/M_MED_DoubleConk_FaceAcc_Physics
    simulationSpace=BaseBoneSpace
    alphaBoolBlend {blendOption=Linear}
    baseBoneRef {boneName=head}
    simSpaceSettings {
     dampingAlpha=0.7
     maxAngularAcceleration=12
     maxAngularVelocity=3
     maxLinearAcceleration=1500
     maxLinearVelocity=400
     worldAlpha=1
     externalLinearDragV=<0.2 0.2 1>
    }
   }
   n3 = Anim Dynamics - Bone: dyn_hairStrand_a!(ComponentPose=@n2, Alpha=1, GravityOverride:Vector="0,0,-980", ExternalForce="0,0,0") -> (Pose) {
    angularDampingOverride=0.99
    angularSpringConstant=4000
    bAngularSpring=true
    bOverrideAngularDamping=true
    bUseGravityOverride=true
    gravityScale=0.35
    alphaBoolBlend {blendOption=Linear}
    boundBone {boneName=dyn_hairStrand_a}
    componentAppliedLinearAccClamp=<1000 1000 1000>
    componentLinearAccScale=<0.3 0.3 0.3>
    componentLinearVelScale=<0.4 0.4 0.4>
    gravityOverride {z=-980}
    physicsBodyDefinitions [{
      boundBone {boneName=dyn_hairStrand_a}
      boxExtents=<20 30 30>
      constraintSetup {angularLimitsMax=<15 15 6>, angularLimitsMin=<-15 -15 -15>, angularTarget=<1 0 0>}
      localJointOffset=<2.4 0 0>
     }]
   }
   n4 = Anim Dynamics - Bone: dyn_hairStrand_b!(ComponentPose=@n3, Alpha=1, GravityOverride:Vector="0,0,-980", ExternalForce="0,0,0") -> (Pose) {
    angularDampingOverride=0.99
    angularSpringConstant=4000
    bAngularSpring=true
    bOverrideAngularDamping=true
    bUseGravityOverride=true
    gravityScale=0.35
    alphaBoolBlend {blendOption=Linear}
    boundBone {boneName=dyn_hairStrand_b}
    componentAppliedLinearAccClamp=<1000 1000 1000>
    componentLinearAccScale=<0.3 0.3 0.3>
    componentLinearVelScale=<0.4 0.4 0.4>
    gravityOverride {z=-980}
    physicsBodyDefinitions [{
      boundBone {boneName=dyn_hairStrand_b}
      boxExtents=<20 30 30>
      constraintSetup {angularLimitsMax=<15 7 6>, angularLimitsMin=<-15 -4 -15>, angularTarget=<1 0 0>}
      localJointOffset=<2.4 0 0>
     }]
   }
   n5 = Anim Dynamics - Bone: dyn_hairStrand_c!(ComponentPose=@n4, Alpha=1, GravityOverride:Vector="0,0,-980", ExternalForce="0,0,0") -> (Pose) {
    angularDampingOverride=0.99
    angularSpringConstant=3000
    bAngularSpring=true
    bOverrideAngularDamping=true
    bUseGravityOverride=true
    gravityScale=0.35
    alphaBoolBlend {blendOption=Linear}
    boundBone {boneName=dyn_hairStrand_c}
    componentAppliedLinearAccClamp=<1000 1000 1000>
    componentLinearAccScale=<0.3 0.3 0.3>
    componentLinearVelScale=<0.4 0.4 0.4>
    gravityOverride {z=-980}
    physicsBodyDefinitions [{
      boundBone {boneName=dyn_hairStrand_c}
      boxExtents=<30 30 30>
      constraintSetup {angularLimitsMax=<15 6 3>, angularLimitsMin=<-15 8 -5>, angularTarget=<1 0 0>}
      localJointOffset=<0.5 0 0>
     }]
   }
   n6 = Anim Dynamics - Bone: dyn_hairStrand_d!(ComponentPose=@n5, Alpha=1, GravityOverride:Vector="0,0,-980", ExternalForce="0,0,0") -> (Pose) {
    angularDampingOverride=0.99
    angularSpringConstant=3500
    bAngularSpring=true
    bOverrideAngularDamping=true
    bUseGravityOverride=true
    gravityScale=0.35
    alphaBoolBlend {blendOption=Linear}
    boundBone {boneName=dyn_hairStrand_d}
    componentAppliedLinearAccClamp=<1000 1000 1000>
    componentLinearAccScale=<0.3 0.3 0.3>
    componentLinearVelScale=<0.4 0.4 0.4>
    gravityOverride {z=-980}
    physicsBodyDefinitions [{
      boundBone {boneName=dyn_hairStrand_d}
      boxExtents=<22 27 27>
      constraintSetup {angularLimitsMax=<13 16 4>, angularLimitsMin=<-11 -10 -14>, angularTarget=<1 0 0>}
      localJointOffset=<2.2 0 0>
     }]
   }
   n7 = Anim Dynamics - Bone: dyn_hairStrand_e!(ComponentPose=@n6, Alpha=1, GravityOverride:Vector="0,0,-980", ExternalForce="0,0,0") -> (Pose) {
    angularDampingOverride=0.99
    angularSpringConstant=4000
    bAngularSpring=true
    bOverrideAngularDamping=true
    bUseGravityOverride=true
    gravityScale=0.35
    alphaBoolBlend {blendOption=Linear}
    boundBone {boneName=dyn_hairStrand_e}
    componentAppliedLinearAccClamp=<1000 1000 1000>
    componentLinearAccScale=<0.3 0.3 0.3>
    componentLinearVelScale=<0.4 0.4 0.4>
    gravityOverride {z=-980}
    physicsBodyDefinitions [{
      boundBone {boneName=dyn_hairStrand_e}
      boxExtents=<30 30 30>
      constraintSetup {angularLimitsMax=<15 8 6>, angularLimitsMin=<-15 -9 -15>, angularTarget=<1 0 0>}
      localJointOffset=<2 0 0>
     }]
   }
   n8 = Anim Dynamics - Bone: dyn_hairStrand_f!(ComponentPose=@n7, Alpha=1, GravityOverride:Vector="0,0,-980", ExternalForce="0,0,0") -> (Pose) {
    angularDampingOverride=0.99
    angularSpringConstant=4000
    bAngularSpring=true
    bOverrideAngularDamping=true
    bUseGravityOverride=true
    gravityScale=0.35
    alphaBoolBlend {blendOption=Linear}
    boundBone {boneName=dyn_hairStrand_f}
    componentAppliedLinearAccClamp=<1000 1000 1000>
    componentLinearAccScale=<0.3 0.3 0.3>
    componentLinearVelScale=<0.4 0.4 0.4>
    gravityOverride {z=-980}
    physicsBodyDefinitions [{
      boundBone {boneName=dyn_hairStrand_f}
      boxExtents=<20 30 30>
      constraintSetup {angularLimitsMax=<15 7 6>, angularLimitsMin=<-15 -6 -15>, angularTarget=<1 0 0>}
      localJointOffset=<2.4 0 0>
     }]
   }
   n9 = Anim Dynamics - Bone: dyn_hairStrand_g!(ComponentPose=@n8, Alpha=1, GravityOverride:Vector="0,0,-980", ExternalForce="0,0,0") -> (Pose) {
    angularDampingOverride=0.99
    angularSpringConstant=3000
    bAngularSpring=true
    bOverrideAngularDamping=true
    bUseGravityOverride=true
    gravityScale=0.35
    alphaBoolBlend {blendOption=Linear}
    boundBone {boneName=dyn_hairStrand_g}
    componentAppliedLinearAccClamp=<1000 1000 1000>
    componentLinearAccScale=<0.3 0.3 0.3>
    componentLinearVelScale=<0.4 0.4 0.4>
    gravityOverride {z=-980}
    physicsBodyDefinitions [{
      boundBone {boneName=dyn_hairStrand_g}
      boxExtents=<30 30 30>
      constraintSetup {angularLimitsMax=<15 15 6>, angularLimitsMin=<-15 -15 -15>, angularTarget=<1 0 0>}
      localJointOffset=<0.6 0 0>
     }]
   }
   n10 = Anim Dynamics - Bone: dyn_hairStrand_h!(ComponentPose=@n9, Alpha=1, GravityOverride:Vector="0,0,-980", ExternalForce="0,0,0") -> (Pose) {
    angularDampingOverride=0.99
    angularSpringConstant=4000
    bAngularSpring=true
    bOverrideAngularDamping=true
    bUseGravityOverride=true
    gravityScale=0.35
    alphaBoolBlend {blendOption=Linear}
    boundBone {boneName=dyn_hairStrand_h}
    componentAppliedLinearAccClamp=<1000 1000 1000>
    componentLinearAccScale=<0.3 0.3 0.3>
    componentLinearVelScale=<0.4 0.4 0.4>
    gravityOverride {z=-980}
    physicsBodyDefinitions [{
      boundBone {boneName=dyn_hairStrand_h}
      boxExtents=<20 30 30>
      constraintSetup {angularLimitsMax=<15 12 4>, angularLimitsMin=<-15 -10 -12>, angularTarget=<1 0 0>}
      localJointOffset=<2 0 0>
     }]
   }
   n11 = Anim Dynamics - Bone: dyn_hairStrand_i!(ComponentPose=@n10, Alpha=1, GravityOverride:Vector="0,0,-980", ExternalForce="0,0,0") -> (Pose) {
    angularDampingOverride=0.99
    angularSpringConstant=3500
    bAngularSpring=true
    bOverrideAngularDamping=true
    bUseGravityOverride=true
    gravityScale=0.35
    alphaBoolBlend {blendOption=Linear}
    boundBone {boneName=dyn_hairStrand_i}
    componentAppliedLinearAccClamp=<1500 1000 1000>
    componentLinearAccScale=<0.3 0.3 0.3>
    componentLinearVelScale=<0.4 0.4 0.4>
    gravityOverride {z=-980}
    physicsBodyDefinitions [{
      boundBone {boneName=dyn_hairStrand_i}
      boxExtents=<32 32 32>
      constraintSetup {angularLimitsMax=<15 12 3>, angularLimitsMin=<-10 13 -12>, angularTarget=<1 0 0>}
      localJointOffset=<1.3 0 0>
     }]
   }
   n12 = Anim Dynamics - Bone: dyn_hairStrand_j!(ComponentPose=@n11, Alpha=1, GravityOverride:Vector="0,0,-980", ExternalForce="0,0,0") -> (Pose) {
    angularDampingOverride=0.99
    angularSpringConstant=3000
    bAngularSpring=true
    bOverrideAngularDamping=true
    bUseGravityOverride=true
    gravityScale=0.35
    alphaBoolBlend {blendOption=Linear}
    boundBone {boneName=dyn_hairStrand_j}
    componentAppliedLinearAccClamp=<1000 1000 1000>
    componentLinearAccScale=<0.3 0.3 0.3>
    componentLinearVelScale=<0.4 0.4 0.4>
    gravityOverride {z=-980}
    physicsBodyDefinitions [{
      boundBone {boneName=dyn_hairStrand_j}
      boxExtents=<25 25 25>
      constraintSetup {angularLimitsMax=<15 15 6>, angularLimitsMin=<-15 -15 -15>, angularTarget=<1 0 0>}
      localJointOffset=<2 0 0>
     }]
   }
   n13 = Anim Dynamics - Bone: dyn_hairFront_e_01!(ComponentPose=@n12, Alpha=1, GravityOverride:Vector="0,0,-980", ExternalForce="0,0,0") -> (Pose) {
    angularDampingOverride=0.99
    angularSpringConstant=2000
    bAngularSpring=true
    bOverrideAngularDamping=true
    bUseGravityOverride=true
    gravityScale=0.35
    alphaBoolBlend {blendOption=Linear}
    boundBone {boneName=dyn_hairFront_e_01}
    componentAppliedLinearAccClamp=<2500 2500 2500>
    componentLinearAccScale=<0.2 0.2 0.2>
    componentLinearVelScale=<0.4 0.3 0.3>
    gravityOverride {z=-980}
    physicsBodyDefinitions [{
      boundBone {boneName=dyn_hairFront_e_01}
      boxExtents=<25 25 25>
      constraintSetup {angularLimitsMax=<10 15 6>, angularLimitsMin=<-7 -15 -15>, angularTarget=<1 0 0>}
      localJointOffset=<2 0 0>
     }]
   }
   n14 = Anim Dynamics - Bone: dyn_hairFront_d_01!(ComponentPose=@n13, Alpha=1, GravityOverride:Vector="0,0,-980", ExternalForce="0,0,0") -> (Pose) {
    angularDampingOverride=0.99
    angularSpringConstant=2500
    bAngularSpring=true
    bOverrideAngularDamping=true
    bUseGravityOverride=true
    gravityScale=0.35
    alphaBoolBlend {blendOption=Linear}
    boundBone {boneName=dyn_hairFront_d_01}
    componentAppliedLinearAccClamp=<2500 2500 2500>
    componentLinearAccScale=<0.2 0.2 0.2>
    componentLinearVelScale=<0.4 0.3 0.3>
    gravityOverride {z=-980}
    physicsBodyDefinitions [{
      boundBone {boneName=dyn_hairFront_d_01}
      boxExtents=<25 25 25>
      constraintSetup {angularLimitsMax=<15 15 6>, angularLimitsMin=<-15 -15 -15>, angularTarget=<1 0 0>}
      localJointOffset=<2.5 0 0>
     }]
   }
   n15 = Anim Dynamics - Bone: dyn_hairFront_c_01!(ComponentPose=@n14, Alpha=1, GravityOverride:Vector="0,0,-980", ExternalForce="0,0,0") -> (Pose) {
    angularDampingOverride=0.99
    angularSpringConstant=2000
    bAngularSpring=true
    bOverrideAngularDamping=true
    bUseGravityOverride=true
    gravityScale=0.35
    alphaBoolBlend {blendOption=Linear}
    boundBone {boneName=dyn_hairFront_c_01}
    componentAppliedLinearAccClamp=<2500 2500 2500>
    componentLinearAccScale=<0.2 0.2 0.2>
    componentLinearVelScale=<0.4 0.3 0.3>
    gravityOverride {z=-980}
    physicsBodyDefinitions [{
      boundBone {boneName=dyn_hairFront_c_01}
      boxExtents=<25 25 25>
      constraintSetup {angularLimitsMax=<15 15 6>, angularLimitsMin=<-15 -15 -15>, angularTarget=<1 0 0>}
      localJointOffset=<3.1 0 0>
     }]
   }
   n16 = Anim Dynamics - Bone: dyn_hairFront_b_01!(ComponentPose=@n15, Alpha=1, GravityOverride:Vector="0,0,-980", ExternalForce="0,0,0") -> (Pose) {
    angularDampingOverride=0.99
    angularSpringConstant=2500
    bAngularSpring=true
    bOverrideAngularDamping=true
    bUseGravityOverride=true
    gravityScale=0.35
    alphaBoolBlend {blendOption=Linear}
    boundBone {boneName=dyn_hairFront_b_01}
    componentAppliedLinearAccClamp=<2500 2500 2500>
    componentLinearAccScale=<0.2 0.2 0.2>
    componentLinearVelScale=<0.4 0.3 0.3>
    gravityOverride {z=-980}
    physicsBodyDefinitions [{
      boundBone {boneName=dyn_hairFront_b_01}
      boxExtents=<25 25 25>
      constraintSetup {angularLimitsMax=<15 15 6>, angularLimitsMin=<-15 -15 -15>, angularTarget=<1 0 0>}
      localJointOffset=<2.5 0 0>
     }]
   }
   n17 = Anim Dynamics - Bone: dyn_hairFront_a_01!(ComponentPose=@n16, Alpha=1, GravityOverride:Vector="0,0,-980", ExternalForce="0,0,0") -> (Pose) {
    angularDampingOverride=0.99
    angularSpringConstant=2000
    bAngularSpring=true
    bOverrideAngularDamping=true
    bUseGravityOverride=true
    gravityScale=0.35
    alphaBoolBlend {blendOption=Linear}
    boundBone {boneName=dyn_hairFront_a_01}
    componentAppliedLinearAccClamp=<2500 2500 2500>
    componentLinearAccScale=<0.2 0.2 0.2>
    componentLinearVelScale=<0.4 0.3 0.3>
    gravityOverride {z=-980}
    physicsBodyDefinitions [{
      boundBone {boneName=dyn_hairFront_a_01}
      boxExtents=<25 25 25>
      constraintSetup {angularLimitsMax=<9 15 6>, angularLimitsMin=<-12 -15 -15>, angularTarget=<1 0 0>}
      localJointOffset=<2 0 0>
     }]
   }
   n18 = Anim Dynamics - Bone: dyn_hairBack_a_04!(ComponentPose=@n17, Alpha=1, GravityOverride:Vector="0,0,-980", ExternalForce="0,0,0") -> (Pose) {
    angularDampingOverride=0.99
    angularSpringConstant=3000
    bAngularSpring=true
    bOverrideAngularDamping=true
    bUseGravityOverride=true
    gravityScale=0.35
    alphaBoolBlend {blendOption=Linear}
    boundBone {boneName=dyn_hairBack_a_04}
    componentAppliedLinearAccClamp=<1000 1000 1000>
    componentLinearAccScale=<0.3 0.3 0.3>
    componentLinearVelScale=<0.4 0.4 0.4>
    gravityOverride {z=-980}
    physicsBodyDefinitions [{
      boundBone {boneName=dyn_hairBack_a_04}
      boxExtents=<30 20 20>
      constraintSetup {angularLimitsMax=<15 0 0>, angularLimitsMin=<-15 -10 -35>, angularTarget=<1 0 0>}
      localJointOffset=<0.5 0 0>
     }]
   }
   n19 = Anim Dynamics - Bone: dyn_hairBack_b_05!(ComponentPose=@n18, Alpha=1, GravityOverride:Vector="0,0,-980", ExternalForce="0,0,0") -> (Pose) {
    angularDampingOverride=0.99
    angularSpringConstant=4000
    bAngularSpring=true
    bOverrideAngularDamping=true
    bUseGravityOverride=true
    gravityScale=0.35
    alphaBoolBlend {blendOption=Linear}
    boundBone {boneName=dyn_hairBack_b_05}
    componentAppliedLinearAccClamp=<1000 1000 1000>
    componentLinearAccScale=<0.3 0.3 0.3>
    componentLinearVelScale=<0.4 0.4 0.4>
    gravityOverride {z=-980}
    physicsBodyDefinitions [{
      boundBone {boneName=dyn_hairBack_b_05}
      boxExtents=<18 20 20>
      constraintSetup {angularLimitsMax=<9 5 0>, angularLimitsMin=<-9 -5 -19>, angularTarget=<1 0 0>}
      localJointOffset=<0.8 0 0>
     }]
   }
   n20 = Anim Dynamics - Bone: dyn_hairBack_b_04!(ComponentPose=@n19, Alpha=1, GravityOverride:Vector="0,0,-980", ExternalForce="0,0,0") -> (Pose) {
    angularDampingOverride=0.99
    angularSpringConstant=4000
    bAngularSpring=true
    bOverrideAngularDamping=true
    bUseGravityOverride=true
    gravityScale=0.35
    alphaBoolBlend {blendOption=Linear}
    boundBone {boneName=dyn_hairBack_b_04}
    componentAppliedLinearAccClamp=<1000 1000 1000>
    componentLinearAccScale=<0.3 0.3 0.3>
    componentLinearVelScale=<0.4 0.4 0.4>
    gravityOverride {z=-980}
    physicsBodyDefinitions [{
      boundBone {boneName=dyn_hairBack_b_04}
      boxExtents=<45 55 55>
      constraintSetup {angularLimitsMax=<8 5 0>, angularLimitsMin=<-8 -5 -18>, angularTarget=<1 0 0>}
      localJointOffset=<0.4 0 0>
     }]
   }
   n21 = Anim Dynamics - Bone: dyn_hairBack_c_04!(ComponentPose=@n20, Alpha=1, GravityOverride:Vector="0,0,-980", ExternalForce="0,0,0") -> (Pose) {
    angularDampingOverride=0.99
    angularSpringConstant=10000
    bAngularSpring=true
    bOverrideAngularDamping=true
    bUseGravityOverride=true
    gravityScale=0.35
    alphaBoolBlend {blendOption=Linear}
    boundBone {boneName=dyn_hairBack_c_04}
    componentAppliedLinearAccClamp=<1000 1000 1000>
    componentLinearAccScale=<0.3 0.3 0.3>
    componentLinearVelScale=<0.4 0.4 0.4>
    gravityOverride {z=-980}
    physicsBodyDefinitions [{
      boundBone {boneName=dyn_hairBack_c_04}
      boxExtents=<55 55 55>
      constraintSetup {angularLimitsMax=<15 0 0>, angularLimitsMin=<-15 -10 -35>, angularTarget=<1 0 0>}
      localJointOffset=<0.5 0 0>
     }]
   }
   n22 = Get bIsSkyDiving()
   n23 = "Transform (Modify) Bone - Bone: dyn_hairBack_b_01"!(ComponentPose=@n21, bAlphaBoolEnabled:bool=@n22, Translation="0,-2,0", Rotation="0,-10,0", Scale="1,1,1") -> (Pose) {
    alphaInputType=Bool
    rotationMode=BMM_Additive
    rotationSpace=BCS_BoneSpace
    translationMode=BMM_Additive
    translationSpace=BCS_BoneSpace
    alphaBoolBlend {blendOption=Linear}
    boneToModify {boneName=dyn_hairBack_b_01}
    rotation {yaw=-10}
    translation {y=-2}
   }
   n24 = "Transform (Modify) Bone - Bone: dyn_hairBack_a_01"!(ComponentPose=@n23, bAlphaBoolEnabled:bool=@n22, Translation="0,-2,0", Rotation="0,-10,0", Scale="1,1,1") -> (Pose) {
    alphaInputType=Bool
    rotationMode=BMM_Additive
    rotationSpace=BCS_BoneSpace
    translationMode=BMM_Additive
    translationSpace=BCS_BoneSpace
    alphaBoolBlend {blendOption=Linear}
    boneToModify {boneName=dyn_hairBack_a_01}
    rotation {yaw=-10}
    translation {y=-2}
   }
   n25 = "Transform (Modify) Bone - Bone: dyn_hairBack_c_01"!(ComponentPose=@n24, bAlphaBoolEnabled:bool=@n22, Translation="0,-2,0", Rotation="0,-10,0", Scale="1,1,1") -> (Pose) {
    alphaInputType=Bool
    rotationMode=BMM_Additive
    rotationSpace=BCS_BoneSpace
    translationMode=BMM_Additive
    translationSpace=BCS_BoneSpace
    alphaBoolBlend {blendOption=Linear}
    boneToModify {boneName=dyn_hairBack_c_01}
    rotation {yaw=-10}
    translation {y=-2}
   }
   Output Pose(Result=Component To Local(ComponentPose=@n25)) {layerGroup=None}
  ]}
 EventGraph {nodes [Event BlueprintUpdateAnimation()]}
]
variables []
